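# Control Rig for the F_MED_ShimPilon character body that adds corrective deformation to body, jacket and strap bones.
# After a shared DeformRig function, thigh spherical pose readers drive lower-jacket offsets, foot yaw drives ankle correctives, and pelvis height drives glute offsets.
# The `baseball` and `knee` booleans apply fixed jacket and strap poses, followed by an always-on strap tweak.
# A neck_03 corrective chain exists but is unconnected to execution.
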ControlRigBlueprint /BRCosmetics/Characters/Player/Female/Medium/Bodies/F_MED_ShimPilon/Meshes/F_MED_ShimPilon_CtrlRig
graphs [
 RigVMFunctionLibrary {links [], nodes []}
 RigVMFunction_Sequence_1_ContainedGraph {nodes [
   n1 = RigVMFunctionEntryNode()
   RigVMFunction_Sequence = &RigVMFunction_Sequence(ExecuteContext=@n1.ExecuteContext)
   n2 = &RigVMFunction_Sequence(ExecuteContext=@RigVMFunction_Sequence.B)
   n3 = &RigVMFunction_Sequence(ExecuteContext=@n2.B)
   n4 = &RigVMFunction_Sequence(ExecuteContext=@n3.B)
   n5 = &RigVMFunction_Sequence(ExecuteContext=@n4.B)
   n6 = &RigVMFunction_Sequence(ExecuteContext=@n5.B)
   RigVMFunctionReturnNode(A=@RigVMFunction_Sequence.A, B=@n2.A, C=@n3.A, D=@n4.A, E=@n5.A, F=@n6.A, G=@n6.B)
  ]}
 RigVMFunction_Sequence_3_ContainedGraph {nodes [
   n1 = RigVMFunctionEntryNode()
   n2 = &RigVMFunction_Sequence(ExecuteContext=@n1.ExecuteContext)
   n3 = &RigVMFunction_Sequence(ExecuteContext=@n2.B)
   RigVMFunctionReturnNode(B=@n3.A, C=@n3.B, A=@n2.A)
  ]}
 RigVMModel {nodes [
   &RigUnit_ParentConstraint(
    AdvancedSettings=(InterpolationType=Average,RotationOrderForFilter=XZY)
    Child=(Type=Bone,Name="neck_03")
    ExecutePin=()
    Parents=((Item=(Type=Bone,Name="neck_02"),Weight=0.65),(Item=(Type=Bone,Name="head"),Weight=0.35))
    Weight=1
    bMaintainOffset=true
   )
   > &RigUnit_ModifyTransforms(
    ItemToModify=((Item=(Type=Bone,Name="None"),Transform=(Rotation=(X=0,Y=0,Z=0,W=1),Translation=(X=0,Y=0,Z=0),Scale3D=(X=1,Y=1,Z=1))),(Item=(Type=Bone,Name="neck_03"),Transform=(Rotation=(X=0,Y=0,Z=0,W=1),Translation=(X=0,Y=0,Z=0),Scale3D=(X=1,Y=1.13,Z=1.13))))
    Weight=1
    WeightMinimum=0
   )
   > RigVMFunction_Sequence_2 = RigVMAggregateNode()
   n1 = &RigUnit_SphericalPoseReader(
    ActiveRegionScaleFactors=(PositiveWidth=0.7,NegativeWidth=0.3,PositiveHeight=1,NegativeHeight=1)
    ActiveRegionSize=0.3
    Debug=(bDrawDebug=true,bDraw2D=False,bDrawLocalAxes=False,DebugScale=25,DebugSegments=20,DebugThickness=0.25)
    DriverAxis=(X=0,Y=1,Z=0)
    DriverItem=(Type=Bone,Name="head")
    FalloffRegionScaleFactors=(PositiveWidth=0.2,NegativeWidth=1.3,PositiveHeight=1,NegativeHeight=1)
    FalloffSize=0.2
    FlipHeightScaling=False
    FlipWidthScaling=False
    OptionalParentItem=(Type=Bone,Name="pelvis")
    RotationOffset=(X=0,Y=5,Z=-90)
    ExecutePin=@RigVMFunction_Sequence_2.A
   )
   > &RigUnit_ModifyTransforms(
    ItemToModify="((Item=(Type=Bone,Name=\"None\"),Transform=(Rotation=(X=0,Y=0,Z=0,W=1),Translation=(X=0,Y=0,Z=0),Scale3D=(X=1,Y=1,Z=1))),(Item=(Type=Bone,Name=\"neck_03\"),Transform=(Rotation=(X=0.0436,Y=0,Z=0,W=0.999),Translation=(X=0.0001,Y=-0.4,Z=-0.1018),Scale3D=(X=1,Y=1. ...(truncated)"
    WeightMinimum=0
    Weight=@n1.OutputParam
   )
   > n2 = &RigUnit_SphericalPoseReader(
    ActiveRegionScaleFactors=(PositiveWidth=0.7,NegativeWidth=0.3,PositiveHeight=1.5,NegativeHeight=1.5)
    ActiveRegionSize=0.3
    Debug=(bDrawDebug=true,bDraw2D=False,bDrawLocalAxes=False,DebugScale=25,DebugSegments=20,DebugThickness=0.25)
    DriverAxis=(X=0,Y=1,Z=0)
    DriverItem=(Type=Bone,Name="head")
    FalloffRegionScaleFactors=(PositiveWidth=0.2,NegativeWidth=1.3,PositiveHeight=1,NegativeHeight=1)
    FalloffSize=0.2
    FlipHeightScaling=False
    FlipWidthScaling=False
    OptionalParentItem=(Type=Bone,Name="pelvis")
    RotationOffset=(X=73,Y=-15,Z=-90)
   )
   > &RigUnit_ModifyTransforms(
    ItemToModify="((Item=(Type=Bone,Name=\"None\"),Transform=(Rotation=(X=0,Y=0,Z=0,W=1),Translation=(X=0,Y=0,Z=0),Scale3D=(X=1,Y=1,Z=1))),(Item=(Type=Bone,Name=\"neck_03\"),Transform=(Rotation=(X=0.0436,Y=0,Z=0,W=-0.999),Translation=(X=0,Y=-0.4239,Z=0.5),Scale3D=(X=1,Y=1.1,Z=1 ...(truncated)"
    WeightMinimum=0
    Weight=@n2.OutputParam
   )
   n3 = &RigUnit_SphericalPoseReader(
    ActiveRegionScaleFactors=(PositiveWidth=1,NegativeWidth=0.3,PositiveHeight=1,NegativeHeight=1)
    ActiveRegionSize=0.3
    Debug=(bDrawDebug=true,bDraw2D=False,bDrawLocalAxes=False,DebugScale=25,DebugSegments=20,DebugThickness=0.25)
    DriverAxis=(X=0,Y=1,Z=0)
    DriverItem=(Type=Bone,Name="head")
    FalloffRegionScaleFactors=(PositiveWidth=0.2,NegativeWidth=1.45,PositiveHeight=1,NegativeHeight=1)
    FalloffSize=0.2
    FlipHeightScaling=False
    FlipWidthScaling=False
    OptionalParentItem=(Type=Bone,Name="pelvis")
    RotationOffset=(X=179.9582,Y=-8,Z=-89.7101)
    ExecutePin=@RigVMFunction_Sequence_2.B
   )
   > &RigUnit_ModifyTransforms(
    ItemToModify="((Item=(Type=Bone,Name=\"None\"),Transform=(Rotation=(X=0,Y=0,Z=0,W=1),Translation=(X=0,Y=0,Z=0),Scale3D=(X=1,Y=1,Z=1))),(Item=(Type=Bone,Name=\"neck_03\"),Transform=(Rotation=(X=-0.1304,Y=-0.0432,Z=-0.0057,W=0.9905),Translation=(X=0,Y=0.8386,Z=-0.6),Scale3D=( ...(truncated)"
    WeightMinimum=0
    Weight=@n3.OutputParam
   )
   > n4 = &RigUnit_SphericalPoseReader(
    ActiveRegionScaleFactors=(PositiveWidth=1,NegativeWidth=0.3,PositiveHeight=1,NegativeHeight=1)
    ActiveRegionSize=0.3
    Debug=(bDrawDebug=true,bDraw2D=False,bDrawLocalAxes=False,DebugScale=25,DebugSegments=20,DebugThickness=0.25)
    DriverAxis=(X=0,Y=1,Z=0)
    DriverItem=(Type=Bone,Name="head")
    FalloffRegionScaleFactors=(PositiveWidth=0.2,NegativeWidth=1.3,PositiveHeight=1,NegativeHeight=1)
    FalloffSize=0.2
    FlipHeightScaling=False
    FlipWidthScaling=False
    OptionalParentItem=(Type=Bone,Name="pelvis")
    RotationOffset=(X=-90.119,Y=0.4909,Z=-99.7124)
   )
   > &RigUnit_ModifyTransforms(
    ItemToModify="((Item=(Type=Bone,Name=\"None\"),Transform=(Rotation=(X=0,Y=0,Z=0,W=1),Translation=(X=0,Y=0,Z=0),Scale3D=(X=1,Y=1,Z=1))),(Item=(Type=Bone,Name=\"neck_03\"),Transform=(Rotation=(X=-0.2588,Y=0,Z=0,W=0.9659),Translation=(X=0,Y=-0.3113,Z=-0.8866),Scale3D=(X=1,Y=1. ...(truncated)"
    WeightMinimum=0
    Weight=@n4.OutputParam
   )
   RigVMFunction_MathQuaternionToEuler = &RigVMFunction_MathQuaternionToEuler(
    RotationOrder=XYZ
    Value=&RigUnit_GetRelativeTransformForItem(Child=(Type=Bone,Name="clavicle_l"), Parent=(Type=Bone,Name="spine_05"), bChildInitial=False, bParentInitial=False).RelativeTransform.Rotation
   )
   Remap = &RigVMFunction_MathDoubleRemap(SourceMaximum=60, SourceMinimum=90, TargetMaximum=1, TargetMinimum=0, Value=@RigVMFunction_MathQuaternionToEuler.Result.Y)
   n5 = &RigUnit_GetRelativeTransformForItem(Child=(Type=Bone,Name="clavicle_r"), Parent=(Type=Bone,Name="spine_05"), bChildInitial=False, bParentInitial=False)
   Multiply = &RigVMFunction_MathDoubleMul(
    A=@Remap.Result
    B=&RigVMFunction_MathDoubleRemap(
     SourceMaximum=60
     SourceMinimum=90
     TargetMaximum=1
     TargetMinimum=0
     Value=&RigVMFunction_MathQuaternionToEuler(RotationOrder=XYZ, Value=@n5.RelativeTransform.Rotation).Result.Y
    ).Result
   )
   &RigUnit_ModifyTransforms(
    ItemToModify=((Item=(Type=Bone,Name="None"),Transform=(Rotation=(X=0,Y=0,Z=0,W=1),Translation=(X=0,Y=0,Z=0),Scale3D=(X=1,Y=1,Z=1))),(Item=(Type=Bone,Name="neck_03"),Transform=(Rotation=(X=0,Y=0,Z=0,W=1),Translation=(X=0,Y=0,Z=0),Scale3D=(X=1,Y=1.35,Z=1.2))))
    WeightMinimum=0
    Weight=@Multiply.Result
    ExecutePin=@RigVMFunction_Sequence_2.C
   )
   RigUnit_BeginExecution = &RigUnit_BeginExecution()
   FN_DeformRig_FNC = RigVMFunctionReferenceNode(
    DeformRig_ Curve="(deformRig_l_upperarm_up_alpha_72_#1=False,deformRig_r_upperarm_up_alpha_73_#2=False,deformRig_l_upperarm_fr_alpha_74_#3=False,deformRig_r_upperarm_fr_alpha_75_CB51B8 ...(truncated)"
    DeformRig_Bool="(upperarm_1_#4=True,elbow_3_#5=True,wrist_7_#6=True,groin_8_#7=True,glute_17_#8=True,knee_9_40C408324CA261 ...(truncated)"
    DeformRig_Debug=(upperarm_debug_18_#4=False,wrist_debug_20_#6=False,lat_debug_26_#9=False,groinGlute_debug_27_#7=False)
    DeformRig_Value="(upperarm_up_18_#10=(Rotation=(X=0,Y=0,Z=0,W=1),Translation=(X=0,Y=0,Z=8),Scale3D=(X=0.7,Y=1,Z=1)),upperarm_fwd_fr_19_#11=(Rotation=(X=0,Y=0,Z=0,W=1),Translation=(X=0,Y=10,Z=0),Scale3D=(X=0.5,Y=1,Z= ...(truncated)"
    ExecuteContext=@RigUnit_BeginExecution.ExecutePin
   )
   > RigVMFunction_Sequence = RigVMAggregateNode()
   n6 = &RigUnit_SphericalPoseReader(
    ActiveRegionScaleFactors=(PositiveWidth=1,NegativeWidth=0.05,PositiveHeight=1,NegativeHeight=1)
    ActiveRegionSize=0.5
    Debug=(bDrawDebug=false,bDraw2D=False,bDrawLocalAxes=False,DebugScale=25,DebugSegments=20,DebugThickness=0.25)
    DriverAxis=(X=1,Y=0,Z=0)
    DriverItem=(Type=Bone,Name="thigh_r")
    FalloffRegionScaleFactors=(PositiveWidth=1,NegativeWidth=1.1,PositiveHeight=1,NegativeHeight=1)
    FalloffSize=0.2
    FlipHeightScaling=False
    FlipWidthScaling=False
    OptionalParentItem=(Type=Bone,Name="pelvis")
    RotationOffset=(X=90,Y=0,Z=178)
    ExecutePin=@RigVMFunction_Sequence.A
   )
   n7 = &RigVMFunction_AnimEvalRichCurve(
    Curve="(EditorCurveData=(Keys=((InterpMode=RCIM_Cubic,ArriveTangent=25.0927),(InterpMode=RCIM_Cubic,Time=0.061,Value=0.168,ArriveTangent=1.0379,LeaveTangent=1.0379),(InterpMode=RCIM_Cubic,TangentMode=RCTM_User,Time=0.2499,Value=0.2594,ArriveTangent=0.8217,LeaveTa ...(truncated)"
    SourceMinimum=0
    TargetMinimum=0
    Value=@n6.OutputParam
   )
   > &RigUnit_ModifyTransforms(
    ItemToModify="((Item=(Type=Bone,Name=\"MSTR_lowjacket1_r\"),Transform=(Rotation=(X=0.3612,Y=-0.0253,Z=0.2807,W=-0.8889),Translation=(X=0.39,Y=-3.954,Z=9.869),Scale3D=(X=1,Y=1,Z=1))),(Item=(Type=Bone,Name=\"MSTR_lowjacket_r\"),Transform=(Rotation=(X=0.1638,Y=0.04,Z=0.1248,W= ...(truncated)"
    WeightMinimum=0
    Weight=@n7.Result
   )
   > n8 = &RigUnit_SphericalPoseReader(
    ActiveRegionScaleFactors=(PositiveWidth=1,NegativeWidth=0.07,PositiveHeight=1,NegativeHeight=1)
    ActiveRegionSize=0.5
    Debug=(bDrawDebug=false,bDraw2D=False,bDrawLocalAxes=False,DebugScale=25,DebugSegments=20,DebugThickness=0.25)
    DriverAxis=(X=-1,Y=0,Z=0)
    DriverItem=(Type=Bone,Name="thigh_l")
    FalloffRegionScaleFactors=(PositiveWidth=1,NegativeWidth=0.9,PositiveHeight=1,NegativeHeight=1)
    FalloffSize=0.2
    FlipHeightScaling=False
    FlipWidthScaling=False
    OptionalParentItem=(Type=Bone,Name="pelvis")
    RotationOffset=(X=90,Y=0,Z=-8)
   )
   n9 = &RigVMFunction_AnimEvalRichCurve(
    Curve="(EditorCurveData=(Keys=((Time=0.031,Value=0.2143,ArriveTangent=4.6254),(InterpMode=RCIM_Cubic,Time=0.5017,Value=0.4932,ArriveTangent=0.8109,LeaveTangent=0.8109),(InterpMode=RCIM_Cubic,TangentMode=RCTM_User,Time=1,Value=1,ArriveTangent=0.9943,LeaveTangent=0 ...(truncated)"
    SourceMinimum=0
    TargetMinimum=0
    Value=@n8.OutputParam
   )
   > &RigUnit_ModifyTransforms(
    ItemToModify="((Item=(Type=Bone,Name=\"MSTR_lowjacket1_l\"),Transform=(Rotation=(X=0.5467,Y=0.0507,Z=-0.2133,W=0.8082),Translation=(X=-3.2323,Y=-4.2078,Z=-8.6379),Scale3D=(X=1,Y=1,Z=1))),(Item=(Type=Bone,Name=\"MSTR_lowjacket_l\"),Transform=(Rotation=(X=0.2961,Y=0.0522,Z=-0 ...(truncated)"
    WeightMinimum=0
    Weight=@n9.Result
   )
   > n10 = &RigUnit_SphericalPoseReader(
    ActiveRegionScaleFactors=(PositiveWidth=1,NegativeWidth=0.1,PositiveHeight=1,NegativeHeight=1)
    ActiveRegionSize=0.5
    Debug=(bDrawDebug=true,bDraw2D=false,bDrawLocalAxes=False,DebugScale=25,DebugSegments=20,DebugThickness=0.25)
    DriverAxis=(X=1,Y=0,Z=0)
    DriverItem=(Type=Bone,Name="thigh_r")
    FalloffRegionScaleFactors=(PositiveWidth=1,NegativeWidth=0.65,PositiveHeight=1,NegativeHeight=1)
    FalloffSize=0.2
    FlipHeightScaling=false
    FlipWidthScaling=false
    OptionalParentItem=(Type=Bone,Name="pelvis")
    RotationOffset=(X=25,Y=17,Z=90)
   )
   > &RigUnit_ModifyTransforms(
    ItemToModify="((Item=(Type=Bone,Name=\"MSTR_lowjacket1_r\"),Transform=(Rotation=(X=0.088,Y=-0.1516,Z=-0.4251,W=0.888),Translation=(X=-4.951,Y=-0.7243,Z=-0.7749),Scale3D=(X=1,Y=1,Z=1))),(Item=(Type=Bone,Name=\"MSTR_lowjacket_r\"),Transform=(Rotation=(X=0,Y=0,Z=0.2164,W=-0.97 ...(truncated)"
    WeightMinimum=0
    Weight=@n10.OutputParam
   )
   > n11 = &RigUnit_SphericalPoseReader(
    ActiveRegionScaleFactors=(PositiveWidth=1,NegativeWidth=0.1,PositiveHeight=1,NegativeHeight=1)
    ActiveRegionSize=0.5
    Debug=(bDrawDebug=true,bDraw2D=false,bDrawLocalAxes=False,DebugScale=25,DebugSegments=20,DebugThickness=0.25)
    DriverAxis=(X=-1,Y=0,Z=0)
    DriverItem=(Type=Bone,Name="thigh_l")
    FalloffRegionScaleFactors=(PositiveWidth=1,NegativeWidth=0.3,PositiveHeight=1,NegativeHeight=1)
    FalloffSize=0.2
    FlipHeightScaling=false
    FlipWidthScaling=false
    OptionalParentItem=(Type=Bone,Name="pelvis")
    RotationOffset=(X=157,Y=-25,Z=-90)
   )
   > &RigUnit_ModifyTransforms(
    ItemToModify="((Item=(Type=Bone,Name=\"MSTR_lowjacket1_l\"),Transform=(Rotation=(X=0,Y=0,Z=0.4617,W=-0.887),Translation=(X=-2.8248,Y=2.605,Z=0),Scale3D=(X=1,Y=1,Z=1))),(Item=(Type=Bone,Name=\"MSTR_lowjacket_l\"),Transform=(Rotation=(X=0.1695,Y=-0.0376,Z=-0.2132,W=0.9615),Tr ...(truncated)"
    WeightMinimum=0
    Weight=@n11.OutputParam
   )
   Get Delta Ref_5 = RigVMFunctionReferenceNode(Item=(Type=Bone,Name="foot_l"), ExecuteContext=@RigVMFunction_Sequence.B)
   > &RigUnit_ModifyTransforms(
    ItemToModify="((Item=(Type=Bone,Name=\"cor_foot_l\"),Transform=(Rotation=(X=-0.605,Y=-0.0658,Z=-0.0595,W=-0.7912),Translation=(X=3.9002,Y=-5.8469,Z=1.6379),Scale3D=(X=1,Y=1,Z=1))),(Item=(Type=Bone,Name=\"cor_foot2_l\"),Transform=(Rotation=(X=-0.2617,Y=0.1251,Z=-0.6095,W=0.7 ...(truncated)"
    WeightMinimum=0
    Weight=&RigVMFunction_MathDoubleRemap(SourceMaximum=60, SourceMinimum=10, TargetMaximum=1, TargetMinimum=0, Value=@"Get Delta Ref_5".Result.Rotation.Yaw).Result
   )
   > &RigUnit_ModifyTransforms(
    ItemToModify=((Item=(Type=Bone,Name="cor_foot_l"),Transform=(Rotation=(X=0,Y=0,Z=0,W=1),Translation=(X=0,Y=0,Z=0),Scale3D=(X=1,Y=1,Z=1))))
    WeightMinimum=0
    Weight=&RigVMFunction_MathDoubleRemap(SourceMaximum=-38, SourceMinimum=-13, TargetMaximum=1, TargetMinimum=0, Value=@"Get Delta Ref_5".Result.Rotation.Yaw).Result
   )
   > Get Delta Ref_2 = RigVMFunctionReferenceNode(Item=(Type=Bone,Name="foot_r"))
   > &RigUnit_ModifyTransforms(
    ItemToModify="((Item=(Type=Bone,Name=\"cor_foot_r\"),Transform=(Rotation=(X=0,Y=0,Z=0,W=1),Translation=(X=3.8508,Y=-6.5867,Z=-1.8),Scale3D=(X=1,Y=1,Z=1))),(Item=(Type=Bone,Name=\"cor_foot2_r\"),Transform=(Rotation=(X=-0.497,Y=0.4079,Z=-0.4859,W=0.5921),Translation=(X=-2.429 ...(truncated)"
    WeightMinimum=0
    Weight=&RigVMFunction_MathDoubleRemap(SourceMaximum=60, SourceMinimum=10, TargetMaximum=1, TargetMinimum=0, Value=@"Get Delta Ref_2".Result.Rotation.Yaw).Result
   )
   > &RigUnit_ModifyTransforms(
    ItemToModify=((Item=(Type=Bone,Name="cor_foot2_r"),Transform=(Rotation=(X=0,Y=0,Z=0,W=1),Translation=(X=0,Y=0,Z=0),Scale3D=(X=1,Y=1,Z=1))))
    WeightMinimum=0
    Weight=&RigVMFunction_MathDoubleRemap(SourceMaximum=-40, SourceMinimum=-13, TargetMaximum=1, TargetMinimum=0, Value=@"Get Delta Ref_2".Result.Rotation.Yaw).Result
   )
   n12 = &RigVMFunction_Sequence(ExecuteContext=@RigVMFunction_Sequence.D)
   RigVMFunction_ControlFlowBranch = &RigVMFunction_ControlFlowBranch(ExecuteContext=@n12.A, Condition=RigVMVariableNode() {variable=baseball}.Value)
   &RigUnit_ModifyTransforms(
    ItemToModify=((Item=(Type=Bone,Name="MSTR_strap_l"),Transform=(Rotation=(X=0.0368,Y=0.079,Z=-0.421,W=0.9029),Translation=(X=0,Y=0,Z=0),Scale3D=(X=1,Y=1,Z=1))))
    Weight=1
    WeightMinimum=0
    ExecutePin=@RigVMFunction_ControlFlowBranch.True
   )
   > &RigUnit_ModifyTransforms(
    ItemToModify="((Item=(Type=Bone,Name=\"MSTR_lowjacket1_r\"),Transform=(Rotation=(X=-0.3007,Y=0,Z=0.3262,W=-0.8962),Translation=(X=-1.625,Y=1.276,Z=0),Scale3D=(X=1,Y=1,Z=1))),(Item=(Type=Bone,Name=\"MSTR_lowjacket_r\"),Transform=(Rotation=(X=0,Y=0,Z=0.7071,W=-0.7071),Transla ...(truncated)"
    Weight=1
    WeightMinimum=0
   )
   > &RigUnit_ModifyTransforms(
    ItemToModify=((Item=(Type=Bone,Name="MSTR_lowjacket2_r"),Transform=(Rotation=(X=0.0038,Y=0.0871,Z=0.673,W=-0.7345),Translation=(X=0,Y=0,Z=0),Scale3D=(X=1,Y=1,Z=1))))
    Weight=1
    WeightMinimum=0
   )
   > &RigUnit_ModifyTransforms(
    ItemToModify="((Item=(Type=Bone,Name=\"MSTR_lowjacket3_r\"),Transform=(Rotation=(X=0,Y=0,Z=0.3007,W=-0.9537),Translation=(X=0.8938,Y=1.2765,Z=1.2894),Scale3D=(X=1,Y=1,Z=1))),(Item=(Type=Bone,Name=\"MSTR_strap_r\"),Transform=(Rotation=(X=0,Y=0,Z=0.7071,W=-0.7071),Translation ...(truncated)"
    Weight=1
    WeightMinimum=0
   )
   n13 = &RigVMFunction_ControlFlowBranch(ExecuteContext=@n12.B, Condition=RigVMVariableNode() {variable=knee}.Value)
   &RigUnit_ModifyTransforms(
    ItemToModify="((Item=(Type=Bone,Name=\"MSTR_lowjacket1_r\"),Transform=(Rotation=(X=0.0008,Y=0.2235,Z=-0.4286,W=-0.8754),Translation=(X=2.7123,Y=2.8747,Z=-1.12),Scale3D=(X=1,Y=1,Z=1))),(Item=(Type=Bone,Name=\"MSTR_lowjacket_r\"),Transform=(Rotation=(X=-0.0342,Y=-0.3527,Z=-0. ...(truncated)"
    Weight=1
    WeightMinimum=0
    ExecutePin=@n13.True
   )
   > &RigUnit_ModifyTransforms(
    ItemToModify="((Item=(Type=Bone,Name=\"MSTR_lowjacket2_r\"),Transform=(Rotation=(X=-0.1409,Y=0.2828,Z=0.3733,W=-0.8722),Translation=(X=0,Y=0,Z=0),Scale3D=(X=1,Y=1,Z=1))),(Item=(Type=Bone,Name=\"MSTR_lowjacket3_r\"),Transform=(Rotation=(X=-0.0454,Y=-0.0416,Z=0.6061,W=-0.793) ...(truncated)"
    Weight=1
    WeightMinimum=0
   )
   > &RigUnit_ModifyTransforms(
    ItemToModify="((Item=(Type=Bone,Name=\"MSTR_lowjacket1_l\"),Transform=(Rotation=(X=0.0358,Y=0.3451,Z=0.0712,W=0.9352),Translation=(X=0.7626,Y=-1.3345,Z=-1.682),Scale3D=(X=1,Y=1,Z=1))),(Item=(Type=Bone,Name=\"MSTR_lowjacket_l\"),Transform=(Rotation=(X=0.074,Y=0.2975,Z=-0.228 ...(truncated)"
    Weight=1
    WeightMinimum=0
   )
   > &RigUnit_ModifyTransforms(
    ItemToModify="((Item=(Type=Bone,Name=\"MSTR_lowjacket2_l\"),Transform=(Rotation=(X=-0.1015,Y=-0.1928,Z=0.4545,W=-0.8637),Translation=(X=0,Y=0,Z=0),Scale3D=(X=1,Y=1,Z=1))),(Item=(Type=Bone,Name=\"MSTR_strap_l\"),Transform=(Rotation=(X=-0.0155,Y=0.0858,Z=-0.5683,W=0.8182),Tra ...(truncated)"
    Weight=1
    WeightMinimum=0
   )
   &RigUnit_ModifyTransforms(
    ItemToModify="((Item=(Type=Bone,Name=\"MSTR_strap_l\"),Transform=(Rotation=(X=0.0436,Y=0,Z=0,W=0.999),Translation=(X=0,Y=0,Z=0),Scale3D=(X=1,Y=1,Z=1))),(Item=(Type=Bone,Name=\"MSTR_strap_r\"),Transform=(Rotation=(X=-0.0436,Y=0,Z=0,W=0.999),Translation=(X=0,Y=0,Z=0),Scale3D= ...(truncated)"
    Weight=1
    WeightMinimum=0
    ExecutePin=@RigVMFunction_Sequence.E
   )
   n14 = &RigVMFunction_MathDoubleMul(
    A=&RigVMFunction_MathDoubleRemap(
     SourceMaximum=45
     SourceMinimum=55
     TargetMaximum=1
     TargetMinimum=0
     Value=&RigUnit_GetRelativeTransformForItem(Child=(Type=Bone,Name="pelvis"), Parent=(Type=Bone,Name="root"), bChildInitial=False, bParentInitial=False).RelativeTransform.Translation.Z
    ).Result
    B=&RigVMFunction_MathDoubleRemap(
     SourceMaximum=40
     SourceMinimum=30
     TargetMaximum=1
     TargetMinimum=0
     Value=&RigUnit_GetRelativeTransformForItem(Child=(Type=Bone,Name="pelvis"), Parent=(Type=Bone,Name="root"), bChildInitial=False, bParentInitial=False).RelativeTransform.Translation.Z
    ).Result
   )
   &RigUnit_ModifyTransforms(
    ItemToModify=((Item=(Type=Bone,Name="deform_glute_l"),Transform=(Rotation=(X=0,Y=0,Z=0,W=1),Translation=(X=-1,Y=0,Z=0),Scale3D=(X=1,Y=1,Z=1))))
    WeightMinimum=0
    Weight=@n14.Result
    ExecutePin=@RigVMFunction_Sequence.C
   )
   > &RigUnit_ModifyTransforms(
    ItemToModify=((Item=(Type=Bone,Name="deform_glute_r"),Transform=(Rotation=(X=0,Y=0,Z=0,W=1),Translation=(X=-10,Y=0,Z=0),Scale3D=(X=1,Y=1,Z=1))))
    WeightMinimum=0
    Weight=@n14.Result
   )
  ]}
]
hierarchy [
 root {loc=<0 0 0>, rot=<0 0 0 1>}
 pelvis {parent=root, loc=<0 -0.5301 98.0571>, rot=<0.0127 0.707 -0.0127 -0.707>}
 spine_01 {parent=pelvis, loc=<6.1208 -0.9328 0>, rot=<0 0 0.034 -0.9994>}
 spine_02 {parent=spine_01, loc=<6.5997 -0.2095 0>, rot=<0 0 0.0238 -0.9997>}
 spine_03 {parent=spine_02, loc=<6.4914 -0.1237 0>, rot=<0 0 0.0149 -0.9999>}
 spine_04 {parent=spine_03, loc=<6.9598 0.0794 0>, rot=<0 0 -0.0132 -0.9999>}
 spine_05 {parent=spine_04, loc=<6.9987 0 0>, rot=<0 0 -0.0015 -1>}
 neck_01 {parent=spine_05, loc=<11.3138 0.1529 0>, rot=<0 0 -0.2003 -0.9797>}
 neck_02 {parent=neck_01, loc=<4.5627 0 0>, rot=<0 0 0.0504 -0.9987>}
 head {parent=neck_02, loc=<5.3914 0 0>, rot=<0 0 0.115 -0.9934>}
 FACIAL_C_FacialRoot {parent=head, loc=<2.0085 0.1746 -0.0002>, rot=<0.5021 -0.4979 -0.5021 -0.4979>}
 FACIAL_C_UnderChin {parent=FACIAL_C_FacialRoot, loc=<-0.0002 8.7048 3.5163>, rot=<0.5 0 0 -0.866>}
 FACIAL_L_UnderChin {parent=FACIAL_C_FacialRoot, loc=<2.7308 7.4531 2.1505>, rot=<0.4025 -0.355 0.0823 -0.8398>}
 FACIAL_R_UnderChin {parent=FACIAL_C_FacialRoot, loc=<-2.7308 7.4531 2.1505>, rot=<0.4025 0.355 -0.0823 -0.8398>}
 FACIAL_C_Jaw {parent=FACIAL_C_FacialRoot, loc=<-0.0002 2.0992 1.2766>, rot=<0.342 0 0 -0.9397>}
 FACIAL_C_Jawline {parent=FACIAL_C_Jaw, loc=<0 2.2387 7.2522>, rot=<0.2588 0 0 -0.9659>}
 FACIAL_L_Jawline {parent=FACIAL_C_FacialRoot, loc=<3.9122 5.4242 4.0145>, rot=<0.2092 0.1578 -0.0547 -0.9635>}
 FACIAL_R_Jawline {parent=FACIAL_C_FacialRoot, loc=<-3.9122 5.4242 4.0145>, rot=<0.2092 -0.1578 0.0547 -0.9635>}
 neck_03 {parent=neck_02, loc=<0 0 0>, rot=<0 0 0 1>}
 clavicle_l {parent=spine_05, loc=<5.4831 -0.1053 -3.2946>, rot=<0.0465 -0.7289 -0.009 -0.6829>}
 upperarm_l {parent=clavicle_l, loc=<9.8658 -1.2136 -0.4105>, rot=<0.0164 -0.3245 -0.0269 -0.9454>}
 lowerarm_l {parent=upperarm_l, loc=<26.465 -0.8345 0.259>, rot=<0.0036 0.0161 -0.2298 -0.9731>}
 hand_l {parent=lowerarm_l, loc=<22.7355 -0.3203 -0.6105>, rot=<-0.6784 -0.0223 0.0735 -0.7307>}
 hand_attach_l {parent=hand_l, loc=<8.5 -3 0>, rot=<0.7071 0.7071 0 0>}
 weapon_l {parent=hand_l, loc=<-0.1974 -3.4111 0.3811>, rot=<0 0 0 1>}
 thumb_01_l {parent=hand_l, loc=<1.3118 -0.3191 -2.459>, rot=<0.6546 -0.4042 0.0664 -0.6354>}
 thumb_02_l {parent=thumb_01_l, loc=<5.9336 -0.0448 0.0401>, rot=<-0.0037 0.0213 0.298 -0.9543>}
 thumb_03_l {parent=thumb_02_l, loc=<2.2223 0.3682 -0.0907>, rot=<0.0118 -0.013 0.0066 -0.9998>}
 index_metacarpal_l {parent=hand_l, loc=<1.6535 0.0183 -2.96>, rot=<0.0158 -0.0124 -0.0173 -0.9996>}
 index_01_l {parent=index_metacarpal_l, loc=<8.3008 0.0357 -0.0524>, rot=<-0.0044 0.0485 0.0924 -0.9945>}
 index_02_l {parent=index_01_l, loc=<3.3587 -0.3004 0.0452>, rot=<-0.002 0.0027 0.3225 -0.9466>}
 index_03_l {parent=index_02_l, loc=<2.5578 -0.123 -0.0252>, rot=<-0.002 0.0067 0.2558 -0.9667>}
 middle_metacarpal_l {parent=hand_l, loc=<2.1856 -0.1725 -0.5984>, rot=<0.006 0.0168 -0.0329 -0.9993>}
 middle_01_l {parent=middle_metacarpal_l, loc=<7.3275 -0.2447 0.0129>, rot=<-0.0208 0.0042 0.0858 -0.9961>}
 middle_02_l {parent=middle_01_l, loc=<4.0123 0.1401 0.1175>, rot=<-0.0218 0.0258 0.3512 -0.9357>}
 middle_03_l {parent=middle_02_l, loc=<2.6077 0.0534 0.1373>, rot=<0.0082 0.0188 0.1708 -0.9851>}
 ring_metacarpal_l {parent=hand_l, loc=<2.2074 -0.3625 1.3638>, rot=<-0.0284 0.0722 -0.0458 -0.9959>}
 ring_01_l {parent=ring_metacarpal_l, loc=<7.0295 -0.6197 -0.1446>, rot=<-0.0489 -0.0252 0.1892 -0.9804>}
 ring_02_l {parent=ring_01_l, loc=<3.6174 0.4647 -0.1027>, rot=<0 0 0.2806 -0.9598>}
 ring_03_l {parent=ring_02_l, loc=<2.4152 0.114 0.0199>, rot=<0 0 0.1189 -0.9929>}
 pinky_metacarpal_l {parent=hand_l, loc=<1.7616 -0.9753 3.3814>, rot=<-0.0397 0.0814 -0.069 -0.9935>}
 pinky_01_l {parent=pinky_metacarpal_l, loc=<7.1709 -0.4831 -0.1094>, rot=<-0.0654 -0.0035 0.2242 -0.9723>}
 pinky_02_l {parent=pinky_01_l, loc=<2.8274 0.0893 -0.1043>, rot=<0.0037 0.0011 0.2854 -0.9584>}
 pinky_03_l {parent=pinky_02_l, loc=<2.2046 0.0307 -0.0078>, rot=<-0.0002 0.0098 0.1445 -0.9895>}
 deform_wrist_dn_base_l {parent=hand_l, loc=<0 -1.5326 0>, rot=<0 0 0 -1>}
 deform_wrist_dn_l {parent=deform_wrist_dn_base_l, loc=<0 0 0>, rot=<0 0 0 1>}
 deform_wrist_up_base_l {parent=hand_l, loc=<0 2.3009 0>, rot=<0 0 0 -1>}
 deform_wrist_up_l {parent=deform_wrist_up_base_l, loc=<0 0 0>, rot=<0 0 0 1>}
 lowerarm_twist_02_l {parent=lowerarm_l, loc=<10.6707 0 0>, rot=<0 0 0 1>}
 lowerarm_twist_01_l {parent=lowerarm_l, loc=<18.316 0 0>, rot=<0 0 0 1>}
 deform_elbow_in_base_l {parent=lowerarm_l, loc=<0.1834 3.6062 -0.2572>, rot=<0 0 0 -1>}
 deform_elbow_in_l {parent=deform_elbow_in_base_l, loc=<0 0 0>, rot=<0 0 0 1>}
 deform_elbow_out_base_l {parent=lowerarm_l, loc=<0 -4 0>, rot=<0 0 0 1>}
 deform_elbow_out_l {parent=deform_elbow_out_base_l, loc=<0 0 0>, rot=<0 0 0 1>}
 upperarm_twist_01_l {parent=upperarm_l, loc=<7.2135 -0.1134 0>, rot=<0 0 0 1>}
 deform_upperarm_bk_base_l {parent=upperarm_twist_01_l, loc=<-6.4053 -4.3746 0.314>, rot=<0 0 0 -1>}
 deform_upperarm_bk_l {parent=deform_upperarm_bk_base_l, loc=<0 0 0>, rot=<0 0 0 1>}
 deform_upperarm_fr_base_l {parent=upperarm_twist_01_l, loc=<-5.6312 5.092 0.3354>, rot=<0 0 0 -1>}
 deform_upperarm_fr_l {parent=deform_upperarm_fr_base_l, loc=<0 0 0>, rot=<0 0 0 1>}
 deform_upperarm_up_base_l {parent=upperarm_twist_01_l, loc=<-8.891 0.4825 3.9517>, rot=<0 0 0 -1>}
 deform_upperarm_up_l {parent=deform_upperarm_up_base_l, loc=<0 0 0>, rot=<0 0 0 1>}
 upperarm_twist_02_l {parent=upperarm_l, loc=<19.7998 -0.1134 0>, rot=<0 0 0 1>}
 clavicle_r {parent=spine_05, loc=<5.4831 -0.1053 3.2946>, rot=<0.7289 0.0465 -0.6829 0.009>}
 upperarm_r {parent=clavicle_r, loc=<-9.8658 1.2136 0.4105>, rot=<0.0164 -0.3245 -0.0269 -0.9454>}
 lowerarm_r {parent=upperarm_r, loc=<-26.465 0.8345 -0.259>, rot=<0.0036 0.0161 -0.2298 -0.9731>}
 hand_r {parent=lowerarm_r, loc=<-22.7355 0.3203 0.6106>, rot=<-0.6784 -0.0223 0.0735 -0.7307>}
 thumb_01_r {parent=hand_r, loc=<-1.3118 0.3191 2.459>, rot=<0.6546 -0.4042 0.0664 -0.6354>}
 thumb_02_r {parent=thumb_01_r, loc=<-5.9336 0.0448 -0.0401>, rot=<-0.0037 0.0213 0.298 -0.9543>}
 thumb_03_r {parent=thumb_02_r, loc=<-2.2223 -0.3682 0.0907>, rot=<0.0118 -0.013 0.0066 -0.9998>}
 index_metacarpal_r {parent=hand_r, loc=<-1.6535 -0.0183 2.96>, rot=<0.0158 -0.0124 -0.0173 -0.9996>}
 index_01_r {parent=index_metacarpal_r, loc=<-8.3008 -0.0357 0.0524>, rot=<-0.0044 0.0485 0.0924 -0.9945>}
 index_02_r {parent=index_01_r, loc=<-3.3587 0.3004 -0.0452>, rot=<-0.002 0.0027 0.3225 -0.9466>}
 index_03_r {parent=index_02_r, loc=<-2.5578 0.123 0.0252>, rot=<-0.002 0.0067 0.2558 -0.9667>}
 middle_metacarpal_r {parent=hand_r, loc=<-2.1856 0.1725 0.5984>, rot=<0.006 0.0168 -0.0329 -0.9993>}
 middle_01_r {parent=middle_metacarpal_r, loc=<-7.3275 0.2447 -0.0129>, rot=<-0.0208 0.0042 0.0858 -0.9961>}
 middle_02_r {parent=middle_01_r, loc=<-4.0123 -0.1401 -0.1175>, rot=<-0.0218 0.0258 0.3512 -0.9357>}
 middle_03_r {parent=middle_02_r, loc=<-2.6077 -0.0534 -0.1373>, rot=<0.0082 0.0188 0.1708 -0.9851>}
 ring_metacarpal_r {parent=hand_r, loc=<-2.2074 0.3625 -1.3638>, rot=<-0.0284 0.0722 -0.0458 -0.9959>}
 ring_01_r {parent=ring_metacarpal_r, loc=<-7.0295 0.6197 0.1446>, rot=<-0.0489 -0.0252 0.1892 -0.9804>}
 ring_02_r {parent=ring_01_r, loc=<-3.6174 -0.4647 0.1027>, rot=<0 0 0.2806 -0.9598>}
 ring_03_r {parent=ring_02_r, loc=<-2.4152 -0.114 -0.0199>, rot=<0 0 0.1189 -0.9929>}
 pinky_metacarpal_r {parent=hand_r, loc=<-1.7616 0.9753 -3.3814>, rot=<-0.0397 0.0814 -0.069 -0.9935>}
 pinky_01_r {parent=pinky_metacarpal_r, loc=<-7.1709 0.4831 0.1094>, rot=<-0.0654 -0.0035 0.2242 -0.9723>}
 pinky_02_r {parent=pinky_01_r, loc=<-2.8274 -0.0893 0.1043>, rot=<0.0037 0.0011 0.2854 -0.9584>}
 pinky_03_r {parent=pinky_02_r, loc=<-2.2046 -0.0307 0.0078>, rot=<-0.0002 0.0098 0.1445 -0.9895>}
 weapon_r {parent=hand_r, loc=<0.1974 3.4111 -0.3811>, rot=<0 0 0 1>}
 hand_attach_r {parent=hand_r, loc=<-8.5001 2.9999 0>, rot=<0 0 -0.7071 -0.7071>}
 deform_wrist_dn_base_r {parent=hand_r, loc=<-0.0001 1.5328 0>, rot=<0 0 0 -1>}
 deform_wrist_dn_r {parent=deform_wrist_dn_base_r, loc=<0 0 0>, rot=<0 0 0 1>}
 deform_wrist_up_base_r {parent=hand_r, loc=<-0.0002 -2.3007 -0.0001>, rot=<0 0 0 -1>}
 deform_wrist_up_r {parent=deform_wrist_up_base_r, loc=<0 0 0>, rot=<0 0 0 1>}
 lowerarm_twist_02_r {parent=lowerarm_r, loc=<-10.6707 0 0>, rot=<0 0 0 -1>}
 lowerarm_twist_01_r {parent=lowerarm_r, loc=<-18.316 0 0>, rot=<0 0 0 -1>}
 deform_elbow_in_base_r {parent=lowerarm_r, loc=<-0.1834 -3.6062 0.2572>, rot=<0 0 0 -1>}
 deform_elbow_in_r {parent=deform_elbow_in_base_r, loc=<0 0 0>, rot=<0 0 0 1>}
 deform_elbow_out_base_r {parent=lowerarm_r, loc=<0 4 -0.0001>, rot=<0 0 0 -1>}
 deform_elbow_out_r {parent=deform_elbow_out_base_r, loc=<0 0 0>, rot=<0 0 0 1>}
 upperarm_twist_01_r {parent=upperarm_r, loc=<-7.2135 0.1134 0>, rot=<0 0 0 -1>}
 deform_upperarm_bk_base_r {parent=upperarm_twist_01_r, loc=<6.4053 4.3746 -0.314>, rot=<0 0 0 -1>}
 deform_upperarm_bk_r {parent=deform_upperarm_bk_base_r, loc=<0 0 0>, rot=<0 0 0 1>}
 deform_upperarm_fr_base_r {parent=upperarm_twist_01_r, loc=<5.6312 -5.092 -0.3354>, rot=<0 0 0 -1>}
 deform_upperarm_fr_r {parent=deform_upperarm_fr_base_r, loc=<0 0 0>, rot=<0 0 0 1>}
 deform_upperarm_up_base_r {parent=upperarm_twist_01_r, loc=<8.891 -0.4825 -3.9517>, rot=<0 0 0 -1>}
 deform_upperarm_up_r {parent=deform_upperarm_up_base_r, loc=<0 0 0>, rot=<0 0 0 1>}
 upperarm_twist_02_r {parent=upperarm_r, loc=<-19.7998 0.1134 0>, rot=<0 0 0 -1>}
 attach_backpack {parent=spine_05, loc=<-0.123 -13.473 0>, rot=<-0.0093 -0.707 0.0093 -0.707>}
 attach_cape {parent=spine_05, loc=<9.017 -11.714 0>, rot=<0.0616 -0.7044 -0.0616 -0.7044>}
 Backpack_BR {parent=spine_05, type=Null, loc=<0 -9.3853 0>, rot=<0 0.7071 0 0.7071>}
 cape {parent=spine_05, type=Null, loc=<5.197 -7.4579 0>, rot=<0 0.7071 0 0.7071>}
 thigh_l {parent=pelvis, loc=<-1.0855 -0.2109 -8.7053>, rot=<0.0009 0.0355 0.0155 -0.9993>}
 calf_l {parent=thigh_l, loc=<-42.5284 -0.0019 -0.7297>, rot=<0.0006 -0.0161 -0.0391 -0.9991>}
 foot_l {parent=calf_l, loc=<-46.7401 -2.5014 -0.592>, rot=<0.0427 -0.0238 0.0353 -0.9982>}
 ball_l {parent=foot_l, loc=<-3.5658 10.8066 -0.1544>, rot=<0.0001 -0.0001 -0.7063 -0.7079>}
 calf_twist_02_l {parent=calf_l, loc=<-19.1546 -1.7943 -0.2638>, rot=<0.0237 0.0095 -0.0035 -0.9997>}
 calf_twist_01_l {parent=calf_l, loc=<-35.8483 -1.9197 -0.5794>, rot=<0.0237 0.0095 -0.0035 -0.9997>}
 cor_foot_l {parent=calf_twist_01_l, loc=<-3.7874 -2.2316 -0.1506>, rot=<0.0988 -0.9951 0 0>}
 cor_foot2_l {parent=calf_twist_01_l, loc=<-6.1651 4.8906 0>, rot=<0.7534 0.6576 0 0>}
 cor_foot_l1 {parent=calf_twist_01_l, loc=<-10.8646 -2.8634 0.6683>, rot=<0.0988 -0.9951 0 0>}
 deform_knee_in_base_l {parent=calf_l, loc=<0.8031 -5.0231 0.2456>, rot=<0 0 0 -1>}
 deform_knee_in_l {parent=deform_knee_in_base_l, loc=<0 0 0>, rot=<0 0 0 1>}
 deform_knee_out_base_l {parent=calf_l, loc=<0 4 0>, rot=<0 0 0 -1>}
 deform_knee_out_l {parent=deform_knee_out_base_l, loc=<0 0 0>, rot=<0 0 0 1>}
 thigh_twist_01_l {parent=thigh_l, loc=<-27.952 0 0>, rot=<0.0013 0 0 -1>}
 deform_groin_base_l {parent=thigh_l, loc=<0.4801 5.6776 -0.2486>, rot=<0 0 0 -1>}
 deform_groin_l {parent=deform_groin_base_l, loc=<0 0 0>, rot=<0 0 0 1>}
 thigh_r {parent=pelvis, loc=<-1.0855 -0.2109 8.7053>, rot=<0.0355 -0.0009 0.9993 0.0155>}
 calf_r {parent=thigh_r, loc=<42.5284 0.0019 0.7297>, rot=<0.0006 -0.0161 -0.0391 -0.9991>}
 foot_r {parent=calf_r, loc=<46.7401 2.5014 0.592>, rot=<0.0427 -0.0238 0.0353 -0.9982>}
 ball_r {parent=foot_r, loc=<3.5658 -10.8066 0.1544>, rot=<0.0001 -0.0001 -0.7063 -0.7079>}
 calf_twist_02_r {parent=calf_r, loc=<19.1546 1.7943 0.2638>, rot=<0.0237 0.0095 -0.0035 -0.9997>}
 calf_twist_01_r {parent=calf_r, loc=<35.8483 1.9197 0.5794>, rot=<0.0237 0.0095 -0.0035 -0.9997>}
 cor_foot_r {parent=calf_twist_01_r, loc=<3.7874 2.2316 0.1506>, rot=<0.9951 0.0988 0 0>}
 cor_foot2_r {parent=calf_twist_01_r, loc=<6.1651 -4.8906 0>, rot=<0.6576 -0.7534 0 0>}
 cor_foot_r1 {parent=calf_twist_01_r, loc=<10.8646 2.8633 -0.6683>, rot=<0.9951 0.0988 0 0>}
 deform_knee_in_base_r {parent=calf_r, loc=<-0.8031 5.0231 -0.2457>, rot=<0 0 0 -1>}
 deform_knee_in_r {parent=deform_knee_in_base_r, loc=<0 0 0>, rot=<0 0 0 1>}
 deform_knee_out_base_r {parent=calf_r, loc=<0 -4 0>, rot=<0 0 0 -1>}
 deform_knee_out_r {parent=deform_knee_out_base_r, loc=<0 0 0>, rot=<0 0 0 1>}
 thigh_twist_01_r {parent=thigh_r, loc=<27.952 0 0>, rot=<0.0013 0 0 -1>}
 deform_groin_base_r {parent=thigh_r, loc=<-0.4801 -5.6776 0.2486>, rot=<0 0 0 -1>}
 deform_groin_r {parent=deform_groin_base_r, loc=<0 0 0>, rot=<0 0 0 1>}
 deform_glute_base_l {parent=pelvis, loc=<-7.3466 -9.0213 -9.1245>, rot=<0.0172 0.031 -0.4477 -0.8935>}
 deform_glute_l {parent=deform_glute_base_l, loc=<0 0 0>, rot=<0 0 0 1>}
 deform_glute_base_r {parent=pelvis, loc=<-7.3464 -9.0213 9.1245>, rot=<0.031 -0.0172 0.8935 -0.4477>}
 deform_glute_r {parent=deform_glute_base_r, loc=<0 0 0>, rot=<0 0 0 1>}
 MSTR_lowjacket3_l {parent=pelvis, loc=<8.2484 -8.3217 -4.4679>, rot=<0.3175 -0.9441 -0.0756 -0.0473>}
 lowjacket3_01_l {parent=MSTR_lowjacket3_l, loc=<0 0 0>, rot=<0 0 0 1>}
 lowjacket3_02_l {parent=lowjacket3_01_l, loc=<5 0 0>, rot=<0 0 0 1>}
 lowjacket3_03_l {parent=lowjacket3_02_l, loc=<5 -0.6162 0>, rot=<0 0 0 1>}
 MSTR_lowjacket3_r {parent=pelvis, loc=<8.2481 -8.3217 4.4679>, rot=<0.3193 -0.9395 0.1195 0.0325>}
 lowjacket3_01_r {parent=MSTR_lowjacket3_r, loc=<0 0 0>, rot=<0 0 0 1>}
 lowjacket3_02_r {parent=lowjacket3_01_r, loc=<4.9999 0 0>, rot=<0 0 0.0613 -0.9981>}
 lowjacket3_03_r {parent=lowjacket3_02_r, loc=<5.0377 0 0>, rot=<0 0 0 1>}
 MSTR_lowjacket2_l {parent=pelvis, loc=<8.6754 -5.2916 -9.1619>, rot=<0.2369 -0.8007 -0.544 -0.0827>}
 lowjacket2_01_l {parent=MSTR_lowjacket2_l, loc=<0 0 0>, rot=<0 0 0 1>}
 lowjacket2_02_l {parent=lowjacket2_01_l, loc=<8 -0.5576 0>, rot=<0 0 0 1>}
 lowjacket2_03_l {parent=lowjacket2_02_l, loc=<8 0.5576 0>, rot=<0 0 0 1>}
 MSTR_lowjacket1_l {parent=pelvis, loc=<3.3414 5.6272 -9.2992>, rot=<-0.2592 0.3296 0.9031 0.0929>}
 lowjacket1_01_l {parent=MSTR_lowjacket1_l, loc=<0 0 0>, rot=<0 0 0 1>}
 lowjacket1_02_l {parent=lowjacket1_01_l, loc=<8 -0.223 0>, rot=<-0.0225 0.0059 -0.0272 -0.9994>}
 lowjacket1_03_l {parent=lowjacket1_02_l, loc=<9 0 0>, rot=<-0.0264 -0.0367 -0.0821 -0.9956>}
 MSTR_lowjacket1_r {parent=pelvis, loc=<3.3416 5.6272 9.2992>, rot=<0.2546 -0.3961 0.8731 0.126>}
 lowjacket1_01_r {parent=MSTR_lowjacket1_r, loc=<0 0 0>, rot=<0 0 0 1>}
 lowjacket1_02_r {parent=lowjacket1_01_r, loc=<8.0033 0 0>, rot=<0 0 -0.0415 -0.9991>}
 lowjacket1_03_r {parent=lowjacket1_02_r, loc=<9 0 0>, rot=<0 0.0484 0 -0.9988>}
 MSTR_strap_l {parent=pelvis, loc=<9.3322 -5.1506 -10.1753>, rot=<0.2946 -0.8608 -0.389 -0.1449>}
 strap_01_l {parent=MSTR_strap_l, loc=<0 0 0>, rot=<0 0 0 1>}
 strap_02_l {parent=strap_01_l, loc=<8.2109 0.5384 -0.0121>, rot=<0.0177 -0.0013 -0.0857 -0.9962>}
 strap_03_l {parent=strap_02_l, loc=<10.9068 1.3215 0>, rot=<0 0 -0.1433 -0.9897>}
 strap_04_l {parent=strap_03_l, loc=<8.8233 0.9743 0.1504>, rot=<-0.0422 0.008 -0.0834 -0.9956>}
 strap_05_l {parent=strap_04_l, loc=<8.3717 0.3931 -0.0161>, rot=<0.0319 -0.0008 -0.0265 -0.9991>}
 strap_06_l {parent=strap_05_l, loc=<6.6823 0.0097 0>, rot=<0 0 0 1>}
 MSTR_strap_r {parent=pelvis, loc=<9.3323 -5.1506 10.1753>, rot=<0.2689 -0.8626 0.4095 0.1264>}
 strap_01_r {parent=MSTR_strap_r, loc=<0 0 0>, rot=<0 0 0 1>}
 strap_02_r {parent=strap_01_r, loc=<8.2289 0 0>, rot=<0 0 -0.0948 -0.9955>}
 strap_03_r {parent=strap_02_r, loc=<10.9441 0.5459 -0.0246>, rot=<0.0223 -0.0031 -0.1379 -0.9902>}
 strap_04_r {parent=strap_03_r, loc=<8.8584 0.2596 -0.015>, rot=<0.0287 -0.0019 -0.0651 -0.9975>}
 strap_05_r {parent=strap_04_r, loc=<8.3792 0.0179 0.0024>, rot=<-0.0661 0.0006 -0.009 -0.9978>}
 strap_06_r {parent=strap_05_r, loc=<6.6823 0 0>, rot=<0 0 0 1>}
 MSTR_lowjacket_r {parent=pelvis, loc=<7.3594 0.2751 11.9921>, rot=<0.195 -0.6455 0.7262 0.134>}
 lowjacket_01_r {parent=MSTR_lowjacket_r, loc=<0 0 0>, rot=<0 0 0 1>}
 lowjacket_02_r {parent=lowjacket_01_r, loc=<9.4422 0 0>, rot=<0 0 -0.0565 -0.9984>}
 lowjacket_03_r {parent=lowjacket_02_r, loc=<8.5919 0 0>, rot=<0 0 0 1>}
 MSTR_lowjacket_l {parent=pelvis, loc=<7.3594 0.2751 -11.9921>, rot=<-0.1925 0.6463 0.7267 0.1312>}
 lowjacket_01_l {parent=MSTR_lowjacket_l, loc=<0 0 0>, rot=<0 0 0 1>}
 lowjacket_02_l {parent=lowjacket_01_l, loc=<9.4419 -0.0741 0>, rot=<0 0 -0.036 -0.9994>}
 lowjacket_03_l {parent=lowjacket_02_l, loc=<8.5871 0.2846 0>, rot=<0 0 -0.0677 -0.9977>}
 MSTR_lowjacket2_r {parent=pelvis, loc=<8.6759 -5.2916 9.1618>, rot=<0.2646 -0.792 0.5408 0.1015>}
 lowjacket2_01_r {parent=MSTR_lowjacket2_r, loc=<0 0 0>, rot=<0 0 0 1>}
 lowjacket2_02_r {parent=lowjacket2_01_r, loc=<8.02 0 0>, rot=<0 0 -0.0695 -0.9976>}
 lowjacket2_03_r {parent=lowjacket2_02_r, loc=<8.0191 0 0>, rot=<0 0 0.0348 -0.9994>}
 root_2 {loc=<0 0 0>, rot=<0 0 0 1>}
]
variables [
 FE {type=bool, default=False}
 baseball {type=bool, default=False}
 knee {type=bool, default=False}
]
