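# Physics asset for the straps and lower-jacket flaps on the F_MED_ShimPilon female cosmetic body.
# MSTR bones hang off the pelvis with free rotation, each heading a chain: six-segment straps per side with tight swing and twist limits, plus jacket chains, including a right-only third.
# Jacket joints rotate freely, except each lowjacket2 chain's locked middle joint.
# Pelvis and leg capsules plus a flat root box supply body and ground shapes.
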PhysicsAsset /BRCosmetics/Characters/Player/Female/Medium/Bodies/F_MED_ShimPilon/Meshes/F_MED_ShimPilon_Physics_Straps_FE
bodies {
 MSTR_lowjacket2_l [{shapeType=Capsule, length=0.5, radius=0.5}]
 MSTR_lowjacket2_r [{shapeType=Capsule, length=0.5, radius=0.5}]
 MSTR_lowjacket3_r [{shapeType=Capsule, length=4, radius=5, shapeName=MSTR_lowjacket3_r_capsule, center=<0 4.8788 -3.0304>, rotation=<rot 0 0.0002 -30>}]
 MSTR_lowjacket_l [{shapeType=Capsule, length=0.5, radius=0.5, shapeName=MSTR_lowjacket_l_capsule}]
 MSTR_lowjacket_r [{shapeType=Capsule, length=0.5, radius=0.5, shapeName=MSTR_lowjacket_r_capsule}]
 MSTR_strap_l [{shapeType=Capsule, length=0.5, radius=0.5, shapeName=MSTR_strap_l_capsule, rotation=<rot 0.0001 0.0001 0>}]
 MSTR_strap_r [{shapeType=Capsule, length=0.5, radius=0.5, shapeName=MSTR_strap_r_capsule, rotation=<rot 0 0 -180>}]
 ball_l [{shapeType=Capsule, length=15.0206, radius=6.2542, center=<-6.1648 1.7084 1.9656>, rotation=<rot 85.2693 -0.0003 -0.0002>}]
 ball_r [{shapeType=Capsule, length=15.0206, radius=6.2542, center=<6.1648 -1.7084 -1.9656>, rotation=<rot 85.2693 -0.0003 179.9998>}]
 calf_l [{shapeType=Capsule, length=33.3937, radius=6.2961, center=<-23.741 -2.3871 0>, rotation=<rot 87.58 4.989 0.0904>}]
 calf_r [{shapeType=Capsule, length=33.3937, radius=6.2961, center=<23.741 2.3871 0>, rotation=<rot 87.58 4.989 -179.9096>}]
 lowjacket2_01_l [{shapeType=Capsule, length=6, radius=6, shapeName=lowjacket2_01_l_capsule, center=<5.5056 5.3939 0.022>, rotation=<rot -32.6094 9.2222 -25.2>}]
 lowjacket2_01_r [{
   shapeType=Capsule
   length=6
   radius=6
   shapeName=lowjacket2_01_r_capsule
   center=<5.5566 5.6052 -0.1604>
   rotation=<rot 32.609 -9.2199 15.0002>
  }]
 lowjacket2_02_l [{shapeType=Capsule, length=10, radius=6, center=<2.2846 5.8239 -0.2911>, rotation=<rot -32.6093 9.2223 -20.2032>}]
 lowjacket2_02_r [{shapeType=Capsule, length=10, radius=6, center=<4.2875 4.1 -3.8599>, rotation=<rot 32.6089 -9.22 17.0001>}]
 lowjacket2_03_l [{
   shapeType=Capsule
   length=10
   radius=6
   shapeName=lowjacket2_03_l_capsule
   center=<0.0017 5.7 0.0133>
   rotation=<rot -32.6081 9.2231 -22.0001>
  }]
 lowjacket2_03_r [{shapeType=Capsule, length=10, radius=6, shapeName=lowjacket2_03_r_capsule, center=<-0.0517 5.7761 0.0237>, rotation=<rot 32.6084 -9.2195 7>}]
 lowjacket3_01_r [{shapeType=Capsule, length=4, radius=5, shapeName=lowjacket3_01_r_capsule, center=<0 2.6496 0.8307>, rotation=<rot 0 0.0002 -30>}]
 lowjacket3_02_r [{shapeType=Capsule, length=4, radius=5, shapeName=lowjacket3_02_r_capsule, center=<0 3.346 -0.3755>, rotation=<rot 0 0.0002 -30>}]
 lowjacket3_03_r [{
   shapeType=Capsule
   length=4
   radius=5
   shapeName=lowjacket3_03_r_capsule
   center=<0.2093 4.9624 -2.6286>
   rotation=<rot 12.9525 -7.6303 -30.8674>
  }]
 lowjacket_01_l [{shapeType=Capsule, length=0.5, radius=0.5, shapeName=lowjacket_01_l_capsule}]
 lowjacket_01_r [{shapeType=Capsule, length=0.5, radius=0.5, shapeName=lowjacket_01_r_capsule}]
 lowjacket_02_l [{shapeType=Capsule, length=0.5, radius=0.5, shapeName=lowjacket_02_l_capsule}]
 lowjacket_02_r [{
   shapeType=Capsule
   length=5
   radius=4.5
   shapeName=lowjacket_02_r_capsule
   center=<1.6242 4.2428 -0.1605>
   rotation=<rot 9.3908 3.4512 25.2837>
  }]
 lowjacket_03_l [{shapeType=Capsule, length=5, radius=5, shapeName=lowjacket_03_l_capsule, center=<1.094 3.885 1.2276>, rotation=<rot -8.1694 -90 22.2567>}]
 lowjacket_03_r [{
   shapeType=Capsule
   length=5
   radius=5
   shapeName=lowjacket_03_r_capsule
   center=<1.4354 4.5759 0.4239>
   rotation=<rot 17.4457 -93.7554 -21.9246>
  }]
 pelvis [{shapeType=Capsule, length=3.4504, radius=9.7935, center=<-1.5327 -0.7361 0.0395>, rotation=<rot 0 171.1354 -173.7094>}]
 root [{shapeType=Box, extentX=200, extentY=200, extentZ=6.8176, shapeName=root_box, center=<0 0 -3.1144>, rotation=<rot 0 -90 0>}]
 strap_01_l [{shapeType=Capsule, length=4, radius=4, shapeName=strap_01_l_capsule, center=<5.5803 -3.5088 0>, rotation=<rot -90 0 0>}]
 strap_01_r [{shapeType=Capsule, length=4, radius=4, shapeName=strap_01_r_capsule, center=<5.5803 -3.2 0>, rotation=<rot -90 0 0>}]
 strap_02_l [{shapeType=Capsule, length=8.664, radius=4, center=<4.9672 -3.7553 0>, rotation=<rot -90 0 0>}]
 strap_02_r [{shapeType=Capsule, length=6.2334, radius=4, center=<4.9672 -3.7858 0>, rotation=<rot -90 0 0>}]
 strap_03_l [{shapeType=Capsule, length=4.0989, radius=4, shapeName=strap_03_l_capsule, center=<4.1003 -3.8358 0>, rotation=<rot -90 10.0002 0>}]
 strap_03_r [{shapeType=Capsule, length=5.1205, radius=4, shapeName=strap_03_r_capsule, center=<4.8801 -3.9 0>, rotation=<rot -90 10.0003 0>}]
 strap_04_l [{shapeType=Capsule, length=2.5, radius=4, center=<4.3017 -2.6 0>, rotation=<rot -90 0 0>}]
 strap_04_r [{shapeType=Capsule, length=2.5, radius=4, center=<4.3017 -2.6 0>, rotation=<rot -90 0 0>}]
 strap_05_l [{shapeType=Capsule, length=3.5, radius=2.3, center=<3.7 0 0>, rotation=<rot 90 0 0>}]
 strap_05_r [{shapeType=Capsule, length=3.5, radius=2.3, center=<3.7 0 0>, rotation=<rot 90 0 0>}]
 strap_06_l [{shapeType=Box, extentX=4.0524, extentY=8, extentZ=10, shapeName=strap_06_l_box, center=<2.9 -0.15 0>, rotation=<rot 0 -90 0.0004>}]
 strap_06_r [{shapeType=Box, extentX=4.6908, extentY=8, extentZ=10, shapeName=strap_06_r_box, center=<2.9 -0.15 0>, rotation=<rot 0 -90 -0.0004>}]
 thigh_l [{shapeType=Capsule, length=30, radius=9.4823, center=<-20.4578 1.0554 0>, rotation=<rot 90 -5.0004 0>}]
 thigh_r [{shapeType=Capsule, length=30, radius=9.4823, center=<20.4578 -1.0554 0>, rotation=<rot 90 174.9996 0>}]
}
constraints [
 {bone1Name=thigh_l, bone2Name=pelvis, swing1LimitDegrees=45, swing2LimitDegrees=45, twistLimitDegrees=45}
 {bone1Name=calf_l, bone2Name=thigh_l, swing1LimitDegrees=45, swing2LimitDegrees=45, twistLimitDegrees=45}
 {bone1Name=ball_l, bone2Name=calf_l, swing1LimitDegrees=45, swing2LimitDegrees=45, twistLimitDegrees=45}
 {bone1Name=thigh_r, bone2Name=pelvis, swing1LimitDegrees=45, swing2LimitDegrees=45, twistLimitDegrees=45}
 {bone1Name=calf_r, bone2Name=thigh_r, swing1LimitDegrees=45, swing2LimitDegrees=45, twistLimitDegrees=45}
 {bone1Name=ball_r, bone2Name=calf_r, swing1LimitDegrees=45, swing2LimitDegrees=45, twistLimitDegrees=45}
 {
  bone1Name=MSTR_lowjacket2_r
  bone2Name=pelvis
  swing1LimitDegrees=45
  swing1Motion=Free
  swing2LimitDegrees=20
  swing2Motion=Free
  twistLimitDegrees=0
  twistMotion=Free
 }
 {bone1Name=strap_02_r, bone2Name=strap_01_r, swing1LimitDegrees=45, swing2LimitDegrees=8, twistLimitDegrees=2}
 {bone1Name=strap_04_r, bone2Name=strap_03_r, swing1LimitDegrees=45, swing2LimitDegrees=8, twistLimitDegrees=0}
 {
  bone1Name=MSTR_lowjacket2_l
  bone2Name=pelvis
  swing1LimitDegrees=45
  swing1Motion=Free
  swing2LimitDegrees=20
  swing2Motion=Free
  twistLimitDegrees=0
  twistMotion=Free
 }
 {
  bone1Name=lowjacket2_02_l
  bone2Name=lowjacket2_01_l
  swing1LimitDegrees=45
  swing1Motion=Locked
  swing2LimitDegrees=45
  swing2Motion=Locked
  twistLimitDegrees=45
  twistMotion=Locked
 }
 {bone1Name=strap_02_l, bone2Name=strap_01_l, swing1LimitDegrees=45, swing2LimitDegrees=8, twistLimitDegrees=0}
 {bone1Name=strap_04_l, bone2Name=strap_03_l, swing1LimitDegrees=45, swing2LimitDegrees=8, twistLimitDegrees=0}
 {bone1Name=strap_01_r, bone2Name=MSTR_strap_r, swing1LimitDegrees=45, swing2LimitDegrees=7, twistLimitDegrees=2}
 {bone1Name=strap_03_r, bone2Name=strap_02_r, swing1LimitDegrees=45, swing2LimitDegrees=8, twistLimitDegrees=2}
 {bone1Name=strap_01_l, bone2Name=MSTR_strap_l, swing1LimitDegrees=45, swing2LimitDegrees=7, twistLimitDegrees=0}
 {bone1Name=strap_03_l, bone2Name=strap_02_l, swing1LimitDegrees=45, swing2LimitDegrees=8, twistLimitDegrees=0}
 {
  bone1Name=pelvis
  bone2Name=root
  swing1LimitDegrees=45
  swing1Motion=Free
  swing2LimitDegrees=45
  swing2Motion=Free
  twistLimitDegrees=45
  twistMotion=Free
 }
 {
  bone1Name=MSTR_strap_l
  bone2Name=pelvis
  swing1LimitDegrees=20
  swing1Motion=Free
  swing2LimitDegrees=5
  swing2Motion=Free
  twistLimitDegrees=0
  twistMotion=Free
 }
 {
  bone1Name=MSTR_strap_r
  bone2Name=pelvis
  swing1LimitDegrees=20
  swing1Motion=Free
  swing2LimitDegrees=5
  swing2Motion=Free
  twistLimitDegrees=0
  twistMotion=Free
 }
 {
  bone1Name=lowjacket2_03_l
  bone2Name=lowjacket2_02_l
  swing1LimitDegrees=45
  swing1Motion=Free
  swing2LimitDegrees=45
  swing2Motion=Free
  twistLimitDegrees=45
  twistMotion=Free
 }
 {
  bone1Name=lowjacket2_01_l
  bone2Name=MSTR_lowjacket2_l
  swing1LimitDegrees=45
  swing1Motion=Free
  swing2LimitDegrees=45
  swing2Motion=Free
  twistLimitDegrees=45
  twistMotion=Free
 }
 {
  bone1Name=lowjacket2_01_r
  bone2Name=MSTR_lowjacket2_r
  swing1LimitDegrees=45
  swing1Motion=Free
  swing2LimitDegrees=45
  swing2Motion=Free
  twistLimitDegrees=45
  twistMotion=Free
 }
 {
  bone1Name=lowjacket2_02_r
  bone2Name=lowjacket2_01_r
  swing1LimitDegrees=45
  swing1Motion=Locked
  swing2LimitDegrees=45
  swing2Motion=Locked
  twistLimitDegrees=45
  twistMotion=Locked
 }
 {
  bone1Name=lowjacket2_03_r
  bone2Name=lowjacket2_02_r
  swing1LimitDegrees=45
  swing1Motion=Free
  swing2LimitDegrees=45
  swing2Motion=Free
  twistLimitDegrees=45
  twistMotion=Free
 }
 {bone1Name=strap_05_r, bone2Name=strap_04_r, swing1LimitDegrees=45, swing2LimitDegrees=8, twistLimitDegrees=0}
 {bone1Name=strap_05_l, bone2Name=strap_04_l, swing1LimitDegrees=45, swing2LimitDegrees=8, twistLimitDegrees=0}
 {bone1Name=strap_06_r, bone2Name=strap_05_r, swing1LimitDegrees=45, swing2LimitDegrees=8, twistLimitDegrees=0}
 {bone1Name=strap_06_l, bone2Name=strap_05_l, swing1LimitDegrees=45, swing2LimitDegrees=8, twistLimitDegrees=0}
 {
  bone1Name=lowjacket_03_r
  bone2Name=lowjacket_02_r
  swing1LimitDegrees=45
  swing1Motion=Free
  swing2LimitDegrees=45
  swing2Motion=Free
  twistLimitDegrees=45
  twistMotion=Free
 }
 {
  bone1Name=lowjacket_02_r
  bone2Name=lowjacket_01_r
  swing1LimitDegrees=45
  swing1Motion=Free
  swing2LimitDegrees=45
  swing2Motion=Free
  twistLimitDegrees=45
  twistMotion=Free
 }
 {
  bone1Name=lowjacket_01_r
  bone2Name=MSTR_lowjacket_r
  swing1LimitDegrees=45
  swing1Motion=Free
  swing2LimitDegrees=45
  swing2Motion=Free
  twistLimitDegrees=45
  twistMotion=Free
 }
 {
  bone1Name=lowjacket_03_l
  bone2Name=lowjacket_02_l
  swing1LimitDegrees=45
  swing1Motion=Free
  swing2LimitDegrees=45
  swing2Motion=Free
  twistLimitDegrees=45
  twistMotion=Free
 }
 {
  bone1Name=lowjacket_02_l
  bone2Name=lowjacket_01_l
  swing1LimitDegrees=45
  swing1Motion=Free
  swing2LimitDegrees=45
  swing2Motion=Free
  twistLimitDegrees=45
  twistMotion=Free
 }
 {
  bone1Name=lowjacket_01_l
  bone2Name=MSTR_lowjacket_l
  swing1LimitDegrees=45
  swing1Motion=Free
  swing2LimitDegrees=45
  swing2Motion=Free
  twistLimitDegrees=45
  twistMotion=Free
 }
 {
  bone1Name=MSTR_lowjacket_l
  bone2Name=pelvis
  swing1LimitDegrees=45
  swing1Motion=Free
  swing2LimitDegrees=45
  swing2Motion=Free
  twistLimitDegrees=45
  twistMotion=Free
 }
 {
  bone1Name=MSTR_lowjacket_r
  bone2Name=pelvis
  swing1LimitDegrees=45
  swing1Motion=Free
  swing2LimitDegrees=45
  swing2Motion=Free
  twistLimitDegrees=45
  twistMotion=Free
 }
 {
  bone1Name=lowjacket3_03_r
  bone2Name=lowjacket3_02_r
  swing1LimitDegrees=45
  swing1Motion=Free
  swing2LimitDegrees=45
  swing2Motion=Free
  twistLimitDegrees=45
  twistMotion=Free
 }
 {
  bone1Name=lowjacket3_02_r
  bone2Name=lowjacket3_01_r
  swing1LimitDegrees=45
  swing1Motion=Free
  swing2LimitDegrees=45
  swing2Motion=Free
  twistLimitDegrees=45
  twistMotion=Free
 }
 {
  bone1Name=lowjacket3_01_r
  bone2Name=MSTR_lowjacket3_r
  swing1LimitDegrees=45
  swing1Motion=Free
  swing2LimitDegrees=45
  swing2Motion=Free
  twistLimitDegrees=45
  twistMotion=Free
 }
 {
  bone1Name=MSTR_lowjacket3_r
  bone2Name=pelvis
  swing1LimitDegrees=45
  swing1Motion=Free
  swing2LimitDegrees=45
  swing2Motion=Free
  twistLimitDegrees=45
  twistMotion=Free
 }
]
