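# Physics asset for the ShimPilon female medium body's straps and lower jacket, apparently for skydiving.
# Capsules on pelvis, legs and ball bones, plus a flat root box, accompany mirrored five-link strap chains and three-link jacket chains, each hung from pelvis via tiny MSTR anchor bodies.
# Strap links have tight swing and near-zero twist limits, while their anchors rotate freely.
# Jacket chains mix free joints with a locked middle joint, with longer capsules on the right.
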PhysicsAsset /BRCosmetics/Characters/Player/Female/Medium/Bodies/F_MED_ShimPilon/Meshes/F_MED_ShimPilon_Physics_Straps_skydivie
bodies {
 MSTR_lowjacket2_l [{shapeType=Capsule, length=0.5, radius=0.5}]
 MSTR_lowjacket2_r [{shapeType=Capsule, length=0.5, radius=0.5}]
 MSTR_strap_l [{shapeType=Capsule, length=0.505, radius=0.505, shapeName=MSTR_strap_l_capsule}]
 MSTR_strap_r [{shapeType=Capsule, length=0.505, radius=0.505, shapeName=MSTR_strap_r_capsule, rotation=<rot 0 0 -180>}]
 ball_l [{shapeType=Capsule, length=15.0206, radius=6.2542, center=<-4.4181 0.3804 0>, rotation=<rot 90 0 0>}]
 ball_r [{shapeType=Capsule, length=15.0206, radius=6.2542, center=<4.4181 -0.3804 0>, rotation=<rot 90 180 0>}]
 calf_l [{shapeType=Capsule, length=33.3937, radius=6.2961, center=<-23.741 -2.3871 0>, rotation=<rot 90 4.9994 0>}]
 calf_r [{shapeType=Capsule, length=33.3937, radius=6.2961, center=<23.741 2.3871 0>, rotation=<rot 90 -175.0006 0>}]
 lowjacket2_01_l [{
   shapeType=Capsule
   length=6
   radius=6
   shapeName=lowjacket2_01_l_capsule
   center=<5.5643 6.6659 0.2802>
   rotation=<rot -32.6094 9.2222 -25.2031>
  }]
 lowjacket2_01_r [{
   shapeType=Capsule
   length=15
   radius=6
   shapeName=lowjacket2_01_r_capsule
   center=<6.9636 5.9749 0.9174>
   rotation=<rot 32.6097 -9.2206 2.9228>
  }]
 lowjacket2_02_l [{shapeType=Capsule, length=6, radius=6, center=<2.3892 6.6634 0.1431>, rotation=<rot -32.6094 9.2222 -25.2031>}]
 lowjacket2_02_r [{shapeType=Capsule, length=15, radius=6, center=<3.6346 5.3648 0.9598>, rotation=<rot 32.6096 -9.2207 2.9227>}]
 lowjacket2_03_l [{shapeType=Capsule, length=6, radius=6, shapeName=lowjacket2_03_l_capsule, center=<0 6.1924 0.1145>, rotation=<rot -32.6094 9.2222 -25.2031>}]
 lowjacket2_03_r [{
   shapeType=Capsule
   length=15
   radius=6
   shapeName=lowjacket2_03_r_capsule
   center=<1.3574 5.5951 0.9341>
   rotation=<rot 32.6096 -9.2207 2.9227>
  }]
 pelvis [{shapeType=Capsule, length=3.4504, radius=9.7935, center=<-1.5327 -0.7361 0.0395>, rotation=<rot 0 171.1354 -173.7094>}]
 root [{shapeType=Box, extentX=200, extentY=200, extentZ=6.8176, shapeName=root_box, center=<0 0 -3.1144>, rotation=<rot 0 -90 0>}]
 strap_01_l [{shapeType=Capsule, length=5.6759, radius=2.6729, shapeName=strap_01_l_capsule, center=<5.7156 -1.6363 0>, rotation=<rot 90 7.1888 0>}]
 strap_01_r [{shapeType=Capsule, length=5.6759, radius=2.6729, shapeName=strap_01_r_capsule, center=<5.7389 -1.95 0>, rotation=<rot 90 179.376 0>}]
 strap_02_l [{shapeType=Capsule, length=6.014, radius=2.9263, center=<5.2501 -1.8489 0>, rotation=<rot 90 0.0009 0>}]
 strap_02_r [{shapeType=Capsule, length=6.014, radius=2.9263, center=<5.2501 -2 0>, rotation=<rot 90 -179.9993 0>}]
 strap_03_l [{shapeType=Capsule, length=6, radius=2.8242, shapeName=strap_03_l_capsule, center=<4.5336 -1.7562 0>, rotation=<rot 90 10.0006 0>}]
 strap_03_r [{shapeType=Capsule, length=6, radius=2.8242, shapeName=strap_03_r_capsule, center=<4.4863 -1.75 0>, rotation=<rot 90 -179.9995 0>}]
 strap_04_l [{shapeType=Capsule, length=5.2207, radius=3.0316, center=<4.9837 -1.7184 0>, rotation=<rot 0 88.6024 -88.7935>}]
 strap_04_r [{shapeType=Capsule, length=5.2207, radius=3.0316, center=<4.9903 -1.45 0>, rotation=<rot 0 88.6023 91.2063>}]
 strap_05_l [{shapeType=Capsule, length=4.0993, radius=4.058, center=<4.5639 -1.4025 0>, rotation=<rot 0 90.6601 -88.5733>}]
 strap_05_r [{shapeType=Capsule, length=4.0993, radius=4.058, center=<4.5639 -1.4025 0>, rotation=<rot 0 90.6601 91.4267>}]
 thigh_l [{shapeType=Capsule, length=30, radius=9.4823, center=<-20.4578 1.0554 0>, rotation=<rot 90 -5.0004 0>}]
 thigh_r [{shapeType=Capsule, length=30, radius=9.4823, center=<20.4578 -1.0554 0>, rotation=<rot 90 174.9996 0>}]
}
constraints [
 {bone1Name=thigh_l, bone2Name=pelvis, swing1LimitDegrees=45, swing2LimitDegrees=45, twistLimitDegrees=45}
 {bone1Name=calf_l, bone2Name=thigh_l, swing1LimitDegrees=45, swing2LimitDegrees=45, twistLimitDegrees=45}
 {bone1Name=ball_l, bone2Name=calf_l, swing1LimitDegrees=45, swing2LimitDegrees=45, twistLimitDegrees=45}
 {bone1Name=thigh_r, bone2Name=pelvis, swing1LimitDegrees=45, swing2LimitDegrees=45, twistLimitDegrees=45}
 {bone1Name=calf_r, bone2Name=thigh_r, swing1LimitDegrees=45, swing2LimitDegrees=45, twistLimitDegrees=45}
 {bone1Name=ball_r, bone2Name=calf_r, swing1LimitDegrees=45, swing2LimitDegrees=45, twistLimitDegrees=45}
 {bone1Name=MSTR_lowjacket2_r, bone2Name=pelvis, swing1LimitDegrees=45, swing2LimitDegrees=20, twistLimitDegrees=0}
 {bone1Name=strap_02_r, bone2Name=strap_01_r, swing1LimitDegrees=20, swing2LimitDegrees=10, twistLimitDegrees=0}
 {bone1Name=strap_04_r, bone2Name=strap_03_r, swing1LimitDegrees=20, swing2LimitDegrees=10, twistLimitDegrees=3}
 {bone1Name=strap_05_r, bone2Name=strap_04_r, swing1LimitDegrees=20, swing2LimitDegrees=10, twistLimitDegrees=3}
 {bone1Name=MSTR_lowjacket2_l, bone2Name=pelvis, swing1LimitDegrees=45, swing2LimitDegrees=20, twistLimitDegrees=0}
 {
  bone1Name=lowjacket2_02_l
  bone2Name=lowjacket2_01_l
  swing1LimitDegrees=45
  swing1Motion=Locked
  swing2LimitDegrees=45
  swing2Motion=Locked
  twistLimitDegrees=45
  twistMotion=Locked
 }
 {bone1Name=strap_02_l, bone2Name=strap_01_l, swing1LimitDegrees=20, swing2LimitDegrees=10, twistLimitDegrees=0}
 {bone1Name=strap_04_l, bone2Name=strap_03_l, swing1LimitDegrees=20, swing2LimitDegrees=10, twistLimitDegrees=3}
 {bone1Name=strap_05_l, bone2Name=strap_04_l, swing1LimitDegrees=20, swing2LimitDegrees=10, twistLimitDegrees=3}
 {bone1Name=strap_01_r, bone2Name=MSTR_strap_r, swing1LimitDegrees=35, swing2LimitDegrees=7, twistLimitDegrees=2}
 {bone1Name=strap_03_r, bone2Name=strap_02_r, swing1LimitDegrees=20, swing2LimitDegrees=10, twistLimitDegrees=3}
 {bone1Name=strap_01_l, bone2Name=MSTR_strap_l, swing1LimitDegrees=35, swing2LimitDegrees=7, twistLimitDegrees=2}
 {bone1Name=strap_03_l, bone2Name=strap_02_l, swing1LimitDegrees=20, swing2LimitDegrees=10, twistLimitDegrees=3}
 {
  bone1Name=pelvis
  bone2Name=root
  swing1LimitDegrees=45
  swing1Motion=Free
  swing2LimitDegrees=45
  swing2Motion=Free
  twistLimitDegrees=45
  twistMotion=Free
 }
 {
  bone1Name=MSTR_strap_l
  bone2Name=pelvis
  swing1LimitDegrees=20
  swing1Motion=Free
  swing2LimitDegrees=5
  swing2Motion=Free
  twistLimitDegrees=0
  twistMotion=Free
 }
 {
  bone1Name=MSTR_strap_r
  bone2Name=pelvis
  swing1LimitDegrees=20
  swing1Motion=Free
  swing2LimitDegrees=5
  swing2Motion=Free
  twistLimitDegrees=0
  twistMotion=Free
 }
 {
  bone1Name=lowjacket2_03_l
  bone2Name=lowjacket2_02_l
  swing1LimitDegrees=45
  swing1Motion=Free
  swing2LimitDegrees=45
  swing2Motion=Free
  twistLimitDegrees=45
  twistMotion=Free
 }
 {
  bone1Name=lowjacket2_01_l
  bone2Name=MSTR_lowjacket2_l
  swing1LimitDegrees=45
  swing1Motion=Free
  swing2LimitDegrees=45
  swing2Motion=Free
  twistLimitDegrees=45
  twistMotion=Free
 }
 {
  bone1Name=lowjacket2_01_r
  bone2Name=MSTR_lowjacket2_r
  swing1LimitDegrees=45
  swing1Motion=Free
  swing2LimitDegrees=45
  swing2Motion=Free
  twistLimitDegrees=45
  twistMotion=Free
 }
 {
  bone1Name=lowjacket2_02_r
  bone2Name=lowjacket2_01_r
  swing1LimitDegrees=45
  swing1Motion=Locked
  swing2LimitDegrees=45
  swing2Motion=Locked
  twistLimitDegrees=45
  twistMotion=Locked
 }
 {
  bone1Name=lowjacket2_03_r
  bone2Name=lowjacket2_02_r
  swing1LimitDegrees=45
  swing1Motion=Free
  swing2LimitDegrees=45
  swing2Motion=Free
  twistLimitDegrees=45
  twistMotion=Free
 }
]
