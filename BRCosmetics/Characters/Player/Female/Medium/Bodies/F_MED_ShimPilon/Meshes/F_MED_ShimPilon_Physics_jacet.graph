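# Physics asset for the lower jacket of the medium female ShimPilon player body, defining its secondary-motion bodies and collision shapes.
# Each side has four capsule chains, each running from a small MSTR anchor on the pelvis through three segments with mostly tight twist limits.
# Angularly free cross-constraints tie neighboring chains near the hem, linking them into a sheet.
# Pelvis, leg and foot shapes plus a flat root box provide body and floor-level volumes.
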PhysicsAsset /BRCosmetics/Characters/Player/Female/Medium/Bodies/F_MED_ShimPilon/Meshes/F_MED_ShimPilon_Physics_jacet
bodies {
 MSTR_lowjacket1_l [{shapeType=Capsule, length=0.505, radius=0.505, shapeName=MSTR_lowjacket1_l_capsule, center=<-1.5229 0 0>, rotation=<rot 0 -90 0>}]
 MSTR_lowjacket1_r [{shapeType=Capsule, length=0.505, radius=0.505, shapeName=MSTR_lowjacket1_r_capsule, center=<-1.5229 0 0>, rotation=<rot 0 -90 0>}]
 MSTR_lowjacket2_l [{shapeType=Capsule, length=0.505, radius=0.505, shapeName=MSTR_lowjacket2_l_capsule, center=<-2.0312 0 0>}]
 MSTR_lowjacket2_r [{shapeType=Capsule, length=0.505, radius=0.505, shapeName=MSTR_lowjacket2_r_capsule, rotation=<rot 0 -90 0>}]
 MSTR_lowjacket3_l [{shapeType=Capsule, length=0.505, radius=0.505, shapeName=MSTR_lowjacket3_l_capsule, center=<-1.619 0 0>, rotation=<rot 0 -90 0>}]
 MSTR_lowjacket3_r [{shapeType=Capsule, length=0.505, radius=0.505, shapeName=MSTR_lowjacket3_r_capsule, center=<-1.619 0 0>, rotation=<rot 0 -90 0>}]
 MSTR_lowjacket_l [{shapeType=Capsule, length=0.505, radius=0.505, shapeName=MSTR_lowjacket_l_capsule, center=<-3.5312 0 0>, rotation=<rot 0 -90 0>}]
 MSTR_lowjacket_r [{shapeType=Capsule, length=0.505, radius=0.505, shapeName=MSTR_lowjacket_r_capsule, center=<-3.5312 0 0>, rotation=<rot 0 -90 0>}]
 calf_l [{
   shapeType=Capsule
   length=39.9015
   radius=8.2517
   shapeName=calf_l_capsule
   center=<-29.2556 -4.3121 2.3134>
   rotation=<rot -0.0312 87.3935 88>
  }]
 calf_r [{
   shapeType=Capsule
   length=40.4707
   radius=7.6824
   shapeName=calf_r_capsule
   center=<28.0239 3.0401 0.3819>
   rotation=<rot -0.0312 87.3935 -92>
  }]
 foot_l [{
   shapeType=Capsule
   length=15.1003
   radius=8.1989
   shapeName=foot_l_capsule
   center=<-7.214 6.6062 3.1574>
   rotation=<rot 0.0836 3.6549 90.4744>
  }]
 foot_r [{
   shapeType=Capsule
   length=16.552
   radius=5.1052
   shapeName=foot_r_capsule
   center=<5.9753 -6.6622 -0.3674>
   rotation=<rot 0.0836 3.6549 -89.5256>
  }]
 lowjacket1_01_l [{shapeType=Capsule, length=5, radius=3.5, shapeName=lowjacket1_01_l_capsule, center=<5.1323 -1.9784 -0.5261>, rotation=<rot 90 5 0>}]
 lowjacket1_01_r [{shapeType=Capsule, length=4, radius=3.5, shapeName=lowjacket1_01_r_capsule, center=<3.8726 -2.4074 0>, rotation=<rot 90 -0.0004 0>}]
 lowjacket1_02_l [{shapeType=Capsule, length=4, radius=3.5, shapeName=lowjacket1_02_l_capsule, center=<5.1508 -1.9325 -0.5261>, rotation=<rot 90 0 0>}]
 lowjacket1_02_r [{shapeType=Capsule, length=4, radius=3.5, shapeName=lowjacket1_02_r_capsule, center=<5.1508 -2.4141 0>, rotation=<rot 90 0 0>}]
 lowjacket1_03_l [{shapeType=Capsule, length=4, radius=3.5, shapeName=lowjacket1_03_l_capsule, center=<5.6985 -1.6 -0.5261>, rotation=<rot 90 0 0>}]
 lowjacket1_03_r [{shapeType=Capsule, length=4, radius=3.5, shapeName=lowjacket1_03_r_capsule, center=<5.5547 -0.7102 1.6916>, rotation=<rot 90 9.9993 0>}]
 lowjacket2_01_l [{shapeType=Capsule, length=6, radius=4, shapeName=lowjacket2_01_l_capsule, center=<2.5227 -4 0.5>, rotation=<rot -8.5181 1.9146 5.0135>}]
 lowjacket2_01_r [{shapeType=Capsule, length=6, radius=4, shapeName=lowjacket2_01_r_capsule, center=<3.0645 -4 -2.2>, rotation=<rot 18.0737 -1.0243 -5.8573>}]
 lowjacket2_02_l [{shapeType=Capsule, length=6, radius=4, shapeName=lowjacket2_02_l_capsule, center=<2.2962 -3 0>, rotation=<rot -28.1228 0.0016 -5.0001>}]
 lowjacket2_02_r [{shapeType=Capsule, length=6, radius=4, shapeName=lowjacket2_02_r_capsule, center=<2.5 -3.3 -0.5>, rotation=<rot 28.1238 0.001 0>}]
 lowjacket2_03_l [{shapeType=Capsule, length=6, radius=4, shapeName=lowjacket2_03_l_capsule, center=<0.8351 -2 -0.7>, rotation=<rot -28.1229 0.0015 -0.0011>}]
 lowjacket2_03_r [{shapeType=Capsule, length=6, radius=4, shapeName=lowjacket2_03_r_capsule, center=<1.7 -2 -0.8813>, rotation=<rot 28.1238 0.001 0.0006>}]
 lowjacket3_01_l [{shapeType=Capsule, length=5, radius=3.5, shapeName=lowjacket3_01_l_capsule, center=<3 -2.9 1.4529>, rotation=<rot -19.9999 0.0013 14.9998>}]
 lowjacket3_01_r [{shapeType=Capsule, length=5, radius=3.5, shapeName=lowjacket3_01_r_capsule, center=<3 -2.9 -1.4529>, rotation=<rot 19.9999 0.0006 -14.9999>}]
 lowjacket3_02_l [{
   shapeType=Capsule
   length=5
   radius=3.5
   shapeName=lowjacket3_02_l_capsule
   center=<2.9539 -2.3 1.45>
   rotation=<rot -19.9999 0.0013 14.9998>
  }]
 lowjacket3_02_r [{
   shapeType=Capsule
   length=5
   radius=3.5
   shapeName=lowjacket3_02_r_capsule
   center=<2.9539 -2.3 -1.45>
   rotation=<rot 19.9999 0.0006 -14.9999>
  }]
 lowjacket3_03_l [{shapeType=Capsule, length=5, radius=3.5, shapeName=lowjacket3_03_l_capsule, center=<3 -2.3 1.4529>, rotation=<rot -19.9999 0.0013 14.9998>}]
 lowjacket3_03_r [{
   shapeType=Capsule
   length=6.5857
   radius=3.5
   shapeName=lowjacket3_03_r_capsule
   center=<3.5561 -1.8643 -2.9807>
   rotation=<rot 10.3235 2.6193 -24.3125>
  }]
 lowjacket_01_l [{shapeType=Capsule, length=5, radius=4.5, shapeName=lowjacket_01_l_capsule, center=<3.4043 -3.5 -0.5>, rotation=<rot -35 0 0>}]
 lowjacket_01_r [{shapeType=Capsule, length=5, radius=4.5, shapeName=lowjacket_01_r_capsule, center=<3.6687 -3.5 0.5>, rotation=<rot 35 0 0>}]
 lowjacket_02_l [{shapeType=Capsule, length=5, radius=4.5, shapeName=lowjacket_02_l_capsule, center=<3.4043 -3 -0.5>, rotation=<rot -35 0 0>}]
 lowjacket_02_r [{shapeType=Capsule, length=5, radius=4.5, shapeName=lowjacket_02_r_capsule, center=<3.6687 -3 0.5>, rotation=<rot 35 0 0>}]
 lowjacket_03_l [{shapeType=Capsule, length=5, radius=4.5, shapeName=lowjacket_03_l_capsule, center=<3.4043 -2.5 -0.5>, rotation=<rot -35 0 0>}]
 lowjacket_03_r [{shapeType=Capsule, length=5, radius=4.5, shapeName=lowjacket_03_r_capsule, center=<3.6687 -2.5 0.5>, rotation=<rot 35 0 0>}]
 pelvis [
  {shapeType=Sphere, radius=9.8, shapeName=pelvis_sphere, center=<-5.6875 -2 4.7>}
  {shapeType=Sphere, radius=9.8, shapeName=pelvis_sphere, center=<-5.6875 -2 -4.7>}
  {shapeType=Capsule, length=6.2, radius=9.5, center=<0 -0.9139 0>}
 ]
 root [{shapeType=Box, extentX=196.4595, extentY=159.1008, extentZ=12.5163, shapeName=root_box, center=<0 0 -7.6568>, rotation=<rot 0 -90 0>}]
 thigh_l [{shapeType=Capsule, length=23.7601, radius=8.5347, center=<-17.0967 -1.1443 -0.0301>, rotation=<rot 84.375 179.998 179.9993>}]
 thigh_r [{shapeType=Capsule, length=23.7601, radius=8.5347, center=<17.0967 1.1443 0.0301>, rotation=<rot 84.375 179.998 -0.0007>}]
}
constraints [
 {bone1Name=thigh_l, bone2Name=pelvis, swing1LimitDegrees=45, swing2LimitDegrees=45, twistLimitDegrees=45}
 {bone1Name=thigh_r, bone2Name=pelvis, swing1LimitDegrees=45, swing2LimitDegrees=45, twistLimitDegrees=45}
 {bone1Name=MSTR_lowjacket2_l, bone2Name=pelvis, swing1LimitDegrees=35, swing2LimitDegrees=30, twistLimitDegrees=0}
 {bone1Name=lowjacket2_01_l, bone2Name=MSTR_lowjacket2_l, swing1LimitDegrees=40, swing2LimitDegrees=20, twistLimitDegrees=0}
 {bone1Name=lowjacket2_02_l, bone2Name=lowjacket2_01_l, swing1LimitDegrees=40, swing2LimitDegrees=20, twistLimitDegrees=0}
 {bone1Name=lowjacket2_03_l, bone2Name=lowjacket2_02_l, swing1LimitDegrees=40, swing2LimitDegrees=20, twistLimitDegrees=0}
 {bone1Name=lowjacket2_02_r, bone2Name=lowjacket2_01_r, swing1LimitDegrees=50, swing2LimitDegrees=20, twistLimitDegrees=2}
 {bone1Name=lowjacket2_03_r, bone2Name=lowjacket2_02_r, swing1LimitDegrees=50, swing2LimitDegrees=20, twistLimitDegrees=2}
 {bone1Name=MSTR_lowjacket_l, bone2Name=pelvis, swing1LimitDegrees=35, swing2LimitDegrees=20, twistLimitDegrees=0}
 {bone1Name=lowjacket_01_l, bone2Name=MSTR_lowjacket_l, swing1LimitDegrees=40, swing2LimitDegrees=10, twistLimitDegrees=0}
 {bone1Name=lowjacket_02_l, bone2Name=lowjacket_01_l, swing1LimitDegrees=40, swing2LimitDegrees=10, twistLimitDegrees=1}
 {bone1Name=lowjacket_03_l, bone2Name=lowjacket_02_l, swing1LimitDegrees=25, swing2LimitDegrees=10, twistLimitDegrees=1}
 {
  bone1Name=MSTR_lowjacket_r
  bone2Name=pelvis
  swing1LimitDegrees=35
  swing2LimitDegrees=20
  swing2Motion=Locked
  twistLimitDegrees=0
  twistMotion=Locked
 }
 {bone1Name=lowjacket_01_r, bone2Name=MSTR_lowjacket_r, swing1LimitDegrees=25, swing2LimitDegrees=10, twistLimitDegrees=0}
 {bone1Name=lowjacket_02_r, bone2Name=lowjacket_01_r, swing1LimitDegrees=25, swing2LimitDegrees=10, twistLimitDegrees=1}
 {bone1Name=lowjacket_03_r, bone2Name=lowjacket_02_r, swing1LimitDegrees=25, swing2LimitDegrees=10, twistLimitDegrees=1}
 {bone1Name=lowjacket2_01_r, bone2Name=MSTR_lowjacket2_r, swing1LimitDegrees=50, swing2LimitDegrees=20, twistLimitDegrees=0}
 {
  bone1Name=MSTR_lowjacket2_r
  bone2Name=pelvis
  swing1LimitDegrees=45
  swing1Motion=Free
  swing2LimitDegrees=45
  swing2Motion=Free
  twistLimitDegrees=45
  twistMotion=Free
 }
 {
  bone1Name=pelvis
  bone2Name=root
  swing1LimitDegrees=45
  swing1Motion=Free
  swing2LimitDegrees=45
  swing2Motion=Free
  twistLimitDegrees=45
  twistMotion=Free
 }
 {
  bone1Name=MSTR_lowjacket3_l
  bone2Name=pelvis
  swing1LimitDegrees=45
  swing1Motion=Free
  swing2LimitDegrees=45
  swing2Motion=Free
  twistLimitDegrees=45
  twistMotion=Free
 }
 {bone1Name=lowjacket3_01_l, bone2Name=MSTR_lowjacket3_l, swing1LimitDegrees=45, swing2LimitDegrees=0, twistLimitDegrees=0}
 {bone1Name=lowjacket3_02_l, bone2Name=lowjacket3_01_l, swing1LimitDegrees=45, swing2LimitDegrees=0, twistLimitDegrees=0}
 {bone1Name=lowjacket3_03_l, bone2Name=lowjacket3_02_l, swing1LimitDegrees=15, swing2LimitDegrees=0, twistLimitDegrees=0}
 {
  bone1Name=MSTR_lowjacket3_r
  bone2Name=pelvis
  swing1LimitDegrees=45
  swing1Motion=Free
  swing2LimitDegrees=45
  swing2Motion=Free
  twistLimitDegrees=45
  twistMotion=Free
 }
 {bone1Name=lowjacket3_01_r, bone2Name=MSTR_lowjacket3_r, swing1LimitDegrees=45, swing2LimitDegrees=0, twistLimitDegrees=0}
 {bone1Name=lowjacket3_02_r, bone2Name=lowjacket3_01_r, swing1LimitDegrees=45, swing2LimitDegrees=0, twistLimitDegrees=0}
 {bone1Name=lowjacket3_03_r, bone2Name=lowjacket3_02_r, swing1LimitDegrees=45, swing2LimitDegrees=0, twistLimitDegrees=0}
 {
  bone1Name=calf_l
  bone2Name=thigh_l
  swing1LimitDegrees=45
  swing1Motion=Free
  swing2LimitDegrees=45
  swing2Motion=Free
  twistLimitDegrees=45
  twistMotion=Free
 }
 {
  bone1Name=calf_r
  bone2Name=thigh_r
  swing1LimitDegrees=45
  swing1Motion=Free
  swing2LimitDegrees=45
  swing2Motion=Free
  twistLimitDegrees=45
  twistMotion=Free
 }
 {
  bone1Name=foot_l
  bone2Name=calf_l
  swing1LimitDegrees=45
  swing1Motion=Free
  swing2LimitDegrees=45
  swing2Motion=Free
  twistLimitDegrees=45
  twistMotion=Free
 }
 {
  bone1Name=foot_r
  bone2Name=calf_r
  swing1LimitDegrees=45
  swing1Motion=Free
  swing2LimitDegrees=45
  swing2Motion=Free
  twistLimitDegrees=45
  twistMotion=Free
 }
 {
  bone1Name=MSTR_lowjacket1_l
  bone2Name=pelvis
  swing1LimitDegrees=45
  swing1Motion=Free
  swing2LimitDegrees=45
  swing2Motion=Free
  twistLimitDegrees=45
  twistMotion=Free
 }
 {bone1Name=lowjacket1_01_l, bone2Name=MSTR_lowjacket1_l, swing1LimitDegrees=35, swing2LimitDegrees=15, twistLimitDegrees=1}
 {bone1Name=lowjacket1_02_l, bone2Name=lowjacket1_01_l, swing1LimitDegrees=35, swing2LimitDegrees=15, twistLimitDegrees=1}
 {bone1Name=lowjacket1_03_l, bone2Name=lowjacket1_02_l, swing1LimitDegrees=35, swing2LimitDegrees=15, twistLimitDegrees=1}
 {
  bone1Name=MSTR_lowjacket1_r
  bone2Name=pelvis
  swing1LimitDegrees=45
  swing1Motion=Free
  swing2LimitDegrees=45
  swing2Motion=Free
  twistLimitDegrees=45
  twistMotion=Free
 }
 {bone1Name=lowjacket1_01_r, bone2Name=MSTR_lowjacket1_r, swing1LimitDegrees=35, swing2LimitDegrees=15, twistLimitDegrees=1}
 {bone1Name=lowjacket1_02_r, bone2Name=lowjacket1_01_r, swing1LimitDegrees=35, swing2LimitDegrees=15, twistLimitDegrees=1}
 {bone1Name=lowjacket1_03_r, bone2Name=lowjacket1_02_r, swing1LimitDegrees=35, swing2LimitDegrees=15, twistLimitDegrees=1}
 {
  bone1Name=lowjacket_03_r
  bone2Name=lowjacket1_03_r
  swing1LimitDegrees=45
  swing1Motion=Free
  swing2LimitDegrees=45
  swing2Motion=Free
  twistLimitDegrees=45
  twistMotion=Free
 }
 {
  bone1Name=lowjacket2_03_r
  bone2Name=lowjacket_03_r
  swing1LimitDegrees=45
  swing1Motion=Free
  swing2LimitDegrees=45
  swing2Motion=Free
  twistLimitDegrees=45
  twistMotion=Free
 }
 {
  bone1Name=lowjacket3_03_r
  bone2Name=lowjacket2_03_r
  swing1LimitDegrees=45
  swing1Motion=Free
  swing2LimitDegrees=45
  swing2Motion=Free
  twistLimitDegrees=45
  twistMotion=Free
 }
 {
  bone1Name=lowjacket3_02_r
  bone2Name=lowjacket2_02_r
  swing1LimitDegrees=45
  swing1Motion=Free
  swing2LimitDegrees=45
  swing2Motion=Free
  twistLimitDegrees=45
  twistMotion=Free
 }
 {
  bone1Name=lowjacket3_03_l
  bone2Name=lowjacket2_03_l
  swing1LimitDegrees=45
  swing1Motion=Free
  swing2LimitDegrees=45
  swing2Motion=Free
  twistLimitDegrees=45
  twistMotion=Free
 }
 {
  bone1Name=lowjacket2_03_l
  bone2Name=lowjacket_03_l
  swing1LimitDegrees=45
  swing1Motion=Free
  swing2LimitDegrees=45
  swing2Motion=Free
  twistLimitDegrees=45
  twistMotion=Free
 }
 {
  bone1Name=lowjacket_03_l
  bone2Name=lowjacket1_03_l
  swing1LimitDegrees=45
  swing1Motion=Free
  swing2LimitDegrees=45
  swing2Motion=Free
  twistLimitDegrees=45
  twistMotion=Free
 }
 {
  bone1Name=lowjacket3_03_l
  bone2Name=lowjacket2_03_l
  swing1LimitDegrees=45
  swing1Motion=Free
  swing2LimitDegrees=45
  swing2Motion=Free
  twistLimitDegrees=45
  twistMotion=Free
 }
 {
  bone1Name=lowjacket2_02_l
  bone2Name=lowjacket_02_l
  swing1LimitDegrees=45
  swing1Motion=Free
  swing2LimitDegrees=45
  swing2Motion=Free
  twistLimitDegrees=45
  twistMotion=Free
 }
]
