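# Physics asset for the lower jacket of the female ShimPilon character, named as a skydive variant.
# Pelvis and thigh shapes form the body, with four three-segment capsule chains per side hanging from small MSTR root bodies constrained to the pelvis.
# Chain joints use tight swing and twist limits, while root limits vary from locked to free.
# Extra free-rotation constraints cross-link neighboring chains, with one more link on the right than the left.
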PhysicsAsset /BRCosmetics/Characters/Player/Female/Medium/Bodies/F_MED_ShimPilon/Meshes/F_MED_ShimPilon_Physics_jacet_skydivie
bodies {
 MSTR_lowjacket1_l [{shapeType=Capsule, length=0.505, radius=0.505, shapeName=MSTR_lowjacket1_l_capsule, center=<-1.9034 0 0>, rotation=<rot 0 -90 0>}]
 MSTR_lowjacket1_r [{shapeType=Capsule, length=0.505, radius=0.505, shapeName=MSTR_lowjacket1_r_capsule, center=<-1.9034 0 0>, rotation=<rot 0 -90 0>}]
 MSTR_lowjacket2_l [{shapeType=Capsule, length=0.505, radius=0.505, shapeName=MSTR_lowjacket2_l_capsule, center=<-2.0312 0 0>}]
 MSTR_lowjacket2_r [{shapeType=Capsule, length=0.505, radius=0.505, shapeName=MSTR_lowjacket2_r_capsule, rotation=<rot 0 -90 0>}]
 MSTR_lowjacket3_l [{shapeType=Capsule, length=0.5, radius=0.5, shapeName=MSTR_lowjacket3_l_capsule, center=<-1.6133 0 0>}]
 MSTR_lowjacket3_r [{shapeType=Capsule, length=0.5, radius=0.5, shapeName=MSTR_lowjacket3_r_capsule, center=<-1.6133 0 0>, rotation=<rot 0 0.0004 0>}]
 MSTR_lowjacket_l [{shapeType=Capsule, length=0.505, radius=0.505, shapeName=MSTR_lowjacket_l_capsule, center=<-3.5312 0 0>, rotation=<rot 0 -90 0>}]
 MSTR_lowjacket_r [{shapeType=Capsule, length=0.505, radius=0.505, shapeName=MSTR_lowjacket_r_capsule, center=<-3.5312 0 0>, rotation=<rot 0 -90 0>}]
 lowjacket1_01_l [{shapeType=Capsule, length=5, radius=4, shapeName=lowjacket1_01_l_capsule, center=<2.5 -3.8 0>, rotation=<rot 90 -11.2502 0>}]
 lowjacket1_01_r [{shapeType=Capsule, length=5, radius=4, shapeName=lowjacket1_01_r_capsule, center=<2.0401 -3.8 0>, rotation=<rot 90 3.7491 0>}]
 lowjacket1_02_l [{shapeType=Capsule, length=5, radius=4, shapeName=lowjacket1_02_l_capsule, center=<2.5 -3.8 0>, rotation=<rot 90 -0.0001 0>}]
 lowjacket1_02_r [{shapeType=Capsule, length=5, radius=4, shapeName=lowjacket1_02_r_capsule, center=<2.5601 -3.8 0>, rotation=<rot 90 -2.8135 0>}]
 lowjacket1_03_l [{shapeType=Capsule, length=5, radius=4, shapeName=lowjacket1_03_l_capsule, center=<4.6829 -2.5 0>, rotation=<rot 90 -0.0001 0>}]
 lowjacket1_03_r [{shapeType=Capsule, length=5, radius=4, shapeName=lowjacket1_03_r_capsule, center=<4.554 -2.5 0>, rotation=<rot 90 7.8118 0>}]
 lowjacket2_01_l [{shapeType=Capsule, length=6, radius=4, shapeName=lowjacket2_01_l_capsule, center=<2.5 -3.6 0.5>, rotation=<rot -8.5181 1.9146 5.0135>}]
 lowjacket2_01_r [{shapeType=Capsule, length=6, radius=4, shapeName=lowjacket2_01_r_capsule, center=<2.5 -3.6 -0.5>, rotation=<rot 28.1242 0.0009 -5.6244>}]
 lowjacket2_02_l [{shapeType=Capsule, length=6, radius=4, shapeName=lowjacket2_02_l_capsule, center=<2.5 -2.9 0.5>, rotation=<rot -28.123 0.0015 -0.0011>}]
 lowjacket2_02_r [{shapeType=Capsule, length=6, radius=4, shapeName=lowjacket2_02_r_capsule, center=<2.5 -2.9 -0.5>, rotation=<rot 28.1244 0.0007 0.0004>}]
 lowjacket2_03_l [{shapeType=Capsule, length=6, radius=4, shapeName=lowjacket2_03_l_capsule, center=<2.5 -2.6 0.5>, rotation=<rot -28.1232 0.0013 -0.001>}]
 lowjacket2_03_r [{shapeType=Capsule, length=6, radius=4, shapeName=lowjacket2_03_r_capsule, center=<2.5 -2.6 -0.5>, rotation=<rot 28.1244 0.0006 0.0004>}]
 lowjacket3_01_l [{shapeType=Capsule, length=3, radius=3.5, shapeName=lowjacket3_01_l_capsule, center=<2.5 -2.8 1.8>, rotation=<rot -11.2499 0.0011 22.4999>}]
 lowjacket3_01_r [{shapeType=Capsule, length=3, radius=3.5, shapeName=lowjacket3_01_r_capsule, center=<2.5 -2.8 -1.8>, rotation=<rot 6.9632 -13.2743 143.9428>}]
 lowjacket3_02_l [{shapeType=Capsule, length=3, radius=3.5, shapeName=lowjacket3_02_l_capsule, center=<2.5 -2.6 1.7>, rotation=<rot -11.2499 0.0011 22.4999>}]
 lowjacket3_02_r [{shapeType=Capsule, length=3, radius=3.5, shapeName=lowjacket3_02_r_capsule, center=<2.5 -2.6 -1.7>, rotation=<rot 6.9633 -13.2743 143.9428>}]
 lowjacket3_03_l [{shapeType=Capsule, length=3, radius=3.5, shapeName=lowjacket3_03_l_capsule, center=<2.5 -2.4 1.7>, rotation=<rot -11.2499 0.0011 22.4999>}]
 lowjacket3_03_r [{shapeType=Capsule, length=3, radius=3.5, shapeName=lowjacket3_03_r_capsule, center=<2.5 -2.4 -1.7>, rotation=<rot 6.9632 -13.2746 143.9428>}]
 lowjacket_01_l [{shapeType=Capsule, length=4, radius=4, shapeName=lowjacket_01_l_capsule, center=<2.5 -3.2917 1>, rotation=<rot -90 -0.0001 0>}]
 lowjacket_01_r [{shapeType=Capsule, length=4, radius=4, shapeName=lowjacket_01_r_capsule, center=<2.5 -2.75 -1>, rotation=<rot -90 0.0001 0>}]
 lowjacket_02_l [{shapeType=Capsule, length=4, radius=4, shapeName=lowjacket_02_l_capsule, center=<2.5 -2.7 0>, rotation=<rot -90 -10.0003 0>}]
 lowjacket_02_r [{shapeType=Capsule, length=4, radius=4, shapeName=lowjacket_02_r_capsule, center=<2.5 -2.7 0>, rotation=<rot -90 -9.9998 0>}]
 lowjacket_03_l [{shapeType=Capsule, length=4, radius=4, shapeName=lowjacket_03_l_capsule, center=<2.5 -2.43 1>, rotation=<rot -90 -0.0001 0>}]
 lowjacket_03_r [{shapeType=Capsule, length=4, radius=4, shapeName=lowjacket_03_r_capsule, center=<2.5 -2.3052 -1>, rotation=<rot -90 0.0001 0>}]
 pelvis [
  {shapeType=Sphere, radius=9.9734, shapeName=pelvis_sphere, center=<-4.6204 -2.4949 5.3007>}
  {shapeType=Sphere, radius=9.9734, shapeName=pelvis_sphere, center=<-5.7017 -2.4795 -5.845>}
  {shapeType=Capsule, length=6.5903, radius=10, center=<0 -1.3897 0>}
 ]
 thigh_l [{shapeType=Capsule, length=25, radius=9, center=<-18.0388 -1.1443 -0.1229>, rotation=<rot 84.375 179.999 179.9998>}]
 thigh_r [{shapeType=Capsule, length=25, radius=9, center=<18.0388 1.1443 0.1229>, rotation=<rot 84.375 179.999 -0.0002>}]
}
constraints [
 {bone1Name=thigh_l, bone2Name=pelvis, swing1LimitDegrees=45, swing2LimitDegrees=45, twistLimitDegrees=45}
 {bone1Name=thigh_r, bone2Name=pelvis, swing1LimitDegrees=45, swing2LimitDegrees=45, twistLimitDegrees=45}
 {
  bone1Name=MSTR_lowjacket1_l
  bone2Name=pelvis
  swing1LimitDegrees=35
  swing1Motion=Locked
  swing2LimitDegrees=10
  swing2Motion=Locked
  twistLimitDegrees=0
 }
 {bone1Name=lowjacket1_01_l, bone2Name=MSTR_lowjacket1_l, swing1LimitDegrees=35, swing2LimitDegrees=10, twistLimitDegrees=0}
 {bone1Name=lowjacket1_02_l, bone2Name=lowjacket1_01_l, swing1LimitDegrees=40, swing2LimitDegrees=10, twistLimitDegrees=0}
 {bone1Name=lowjacket1_03_l, bone2Name=lowjacket1_02_l, swing1LimitDegrees=20, swing2LimitDegrees=20, twistLimitDegrees=0}
 {
  bone1Name=MSTR_lowjacket1_r
  bone2Name=pelvis
  swing1LimitDegrees=35
  swing1Motion=Locked
  swing2LimitDegrees=30
  swing2Motion=Locked
  twistLimitDegrees=0
  twistMotion=Locked
 }
 {bone1Name=lowjacket1_01_r, bone2Name=MSTR_lowjacket1_r, swing1LimitDegrees=35, swing2LimitDegrees=10, twistLimitDegrees=3}
 {bone1Name=lowjacket1_02_r, bone2Name=lowjacket1_01_r, swing1LimitDegrees=35, swing2LimitDegrees=10, twistLimitDegrees=3}
 {bone1Name=lowjacket1_03_r, bone2Name=lowjacket1_02_r, swing1LimitDegrees=20, swing2LimitDegrees=20, twistLimitDegrees=3}
 {bone1Name=MSTR_lowjacket2_l, bone2Name=pelvis, swing1LimitDegrees=35, swing2LimitDegrees=30, twistLimitDegrees=0}
 {bone1Name=lowjacket2_01_l, bone2Name=MSTR_lowjacket2_l, swing1LimitDegrees=10, swing2LimitDegrees=10, twistLimitDegrees=0}
 {bone1Name=lowjacket2_02_l, bone2Name=lowjacket2_01_l, swing1LimitDegrees=10, swing2LimitDegrees=10, twistLimitDegrees=0}
 {bone1Name=lowjacket2_03_l, bone2Name=lowjacket2_02_l, swing1LimitDegrees=10, swing2LimitDegrees=10, twistLimitDegrees=0}
 {bone1Name=lowjacket2_02_r, bone2Name=lowjacket2_01_r, swing1LimitDegrees=10, swing2LimitDegrees=10, twistLimitDegrees=0}
 {bone1Name=lowjacket2_03_r, bone2Name=lowjacket2_02_r, swing1LimitDegrees=10, swing2LimitDegrees=10, twistLimitDegrees=0}
 {
  bone1Name=MSTR_lowjacket3_l
  bone2Name=pelvis
  swing1LimitDegrees=35
  swing1Motion=Free
  swing2LimitDegrees=10
  swing2Motion=Free
  twistLimitDegrees=0
  twistMotion=Free
 }
 {bone1Name=lowjacket3_01_l, bone2Name=MSTR_lowjacket3_l, swing1LimitDegrees=20, swing2LimitDegrees=20, twistLimitDegrees=0}
 {bone1Name=lowjacket3_02_l, bone2Name=lowjacket3_01_l, swing1LimitDegrees=20, swing2LimitDegrees=10, twistLimitDegrees=0}
 {bone1Name=lowjacket3_03_l, bone2Name=lowjacket3_02_l, swing1LimitDegrees=20, swing2LimitDegrees=20, twistLimitDegrees=0}
 {
  bone1Name=MSTR_lowjacket3_r
  bone2Name=pelvis
  swing1LimitDegrees=35
  swing1Motion=Free
  swing2LimitDegrees=10
  swing2Motion=Free
  twistLimitDegrees=0
  twistMotion=Free
 }
 {bone1Name=lowjacket3_01_r, bone2Name=MSTR_lowjacket3_r, swing1LimitDegrees=20, swing2LimitDegrees=20, twistLimitDegrees=0}
 {bone1Name=lowjacket3_02_r, bone2Name=lowjacket3_01_r, swing1LimitDegrees=20, swing2LimitDegrees=10, twistLimitDegrees=0}
 {bone1Name=lowjacket3_03_r, bone2Name=lowjacket3_02_r, swing1LimitDegrees=20, swing2LimitDegrees=20, twistLimitDegrees=0}
 {bone1Name=MSTR_lowjacket_l, bone2Name=pelvis, swing1LimitDegrees=35, swing2LimitDegrees=20, twistLimitDegrees=0}
 {bone1Name=lowjacket_01_l, bone2Name=MSTR_lowjacket_l, swing1LimitDegrees=10, swing2LimitDegrees=10, twistLimitDegrees=0}
 {bone1Name=lowjacket_02_l, bone2Name=lowjacket_01_l, swing1LimitDegrees=10, swing2LimitDegrees=10, twistLimitDegrees=0}
 {bone1Name=lowjacket_03_l, bone2Name=lowjacket_02_l, swing1LimitDegrees=10, swing2LimitDegrees=10, twistLimitDegrees=0}
 {bone1Name=MSTR_lowjacket_r, bone2Name=pelvis, swing1LimitDegrees=35, swing2LimitDegrees=20, twistLimitDegrees=0}
 {bone1Name=lowjacket_01_r, bone2Name=MSTR_lowjacket_r, swing1LimitDegrees=10, swing2LimitDegrees=10, twistLimitDegrees=0}
 {bone1Name=lowjacket_02_r, bone2Name=lowjacket_01_r, swing1LimitDegrees=10, swing2LimitDegrees=10, twistLimitDegrees=0}
 {bone1Name=lowjacket_03_r, bone2Name=lowjacket_02_r, swing1LimitDegrees=10, swing2LimitDegrees=10, twistLimitDegrees=0}
 {
  bone1Name=lowjacket3_03_r
  bone2Name=lowjacket2_03_r
  swing1LimitDegrees=45
  swing1Motion=Free
  swing2LimitDegrees=45
  swing2Motion=Free
  twistLimitDegrees=45
  twistMotion=Free
 }
 {
  bone1Name=lowjacket3_03_l
  bone2Name=lowjacket2_02_l
  swing1LimitDegrees=45
  swing1Motion=Free
  swing2LimitDegrees=45
  swing2Motion=Free
  twistLimitDegrees=45
  twistMotion=Free
 }
 {
  bone1Name=lowjacket_02_r
  bone2Name=lowjacket1_02_r
  swing1LimitDegrees=45
  swing1Motion=Free
  swing2LimitDegrees=45
  swing2Motion=Free
  twistLimitDegrees=45
  twistMotion=Free
 }
 {
  bone1Name=lowjacket_02_l
  bone2Name=lowjacket1_02_l
  swing1LimitDegrees=45
  swing1Motion=Free
  swing2LimitDegrees=45
  swing2Motion=Free
  twistLimitDegrees=45
  twistMotion=Free
 }
 {bone1Name=lowjacket2_01_r, bone2Name=MSTR_lowjacket2_r, swing1LimitDegrees=10, swing2LimitDegrees=10, twistLimitDegrees=0}
 {
  bone1Name=MSTR_lowjacket2_r
  bone2Name=pelvis
  swing1LimitDegrees=45
  swing1Motion=Free
  swing2LimitDegrees=45
  swing2Motion=Free
  twistLimitDegrees=45
  twistMotion=Free
 }
 {
  bone1Name=lowjacket2_02_r
  bone2Name=lowjacket_02_r
  swing1LimitDegrees=45
  swing1Motion=Free
  swing2LimitDegrees=45
  swing2Motion=Free
  twistLimitDegrees=45
  twistMotion=Free
 }
]
