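# Post-process style Anim Blueprint for the F_MED_ShimPilon character body, adding simulated jacket and strap motion.
# The parent mesh pose runs through a Control Rig fed baseball and knee slide flags, then is cached.
# Three chained RigidBody branches (front end with strap bone offsets, skydive/parachute, default) use separate physics assets, selected by bool blends.
# Each update sets the skydive flag and computes gravity override from a parameter asset.
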AnimBlueprint /BRCosmetics/Characters/Player/Female/Medium/Bodies/F_MED_ShimPilon/Meshes/F_MED_ShimPilon_AnimBP
target_skeleton=&/F_MED_ShimPilon_Skeleton
graphs [
 AnimGraph {nodes [
   n1 = Get GravityOverride()
   n2 = RigidBody!(ComponentPose=Local To Component(LocalPose=Use cached pose 'Gr/sky_1'()), Alpha, OverrideWorldGravity:Vector=@n1, ExternalForce) -> (Pose) {
    bOverrideWorldGravity=true
    overridePhysicsAsset=&/F_MED_ShimPilon_Physics_jacet
    simulationSpace=BaseBoneSpace
    baseBoneRef {boneName=pelvis}
    overrideWorldGravity {z=-980}
    simSpaceSettings {dampingAlpha=0, maxAngularAcceleration=180, maxAngularVelocity=3, maxLinearAcceleration=200, maxLinearVelocity=250, worldAlpha=1}
   }
   n3 = RigidBody(ComponentPose=@n2, OverrideWorldGravity=@n1) {
    bOverrideWorldGravity=true
    overridePhysicsAsset=&/F_MED_ShimPilon_Physics_Straps
    simulationSpace=BaseBoneSpace
    baseBoneRef {boneName=pelvis}
    overrideWorldGravity {z=-980}
    simSpaceSettings {dampingAlpha=0, maxAngularAcceleration=40, maxAngularVelocity=4, maxLinearAcceleration=200, maxLinearVelocity=250, worldAlpha=1}
   }
   n4 = Component To Local(ComponentPose=@n3)
   n5 = Get GravityOverride()
   n6 = RigidBody(ComponentPose=Local To Component(LocalPose=Use cached pose 'Gr/sky_1'()), OverrideWorldGravity=@n5) {
    bOverrideWorldGravity=true
    overridePhysicsAsset=&/F_MED_ShimPilon_Physics_jacet_skydivie
    simulationSpace=BaseBoneSpace
    baseBoneRef {boneName=pelvis}
    overrideWorldGravity {z=-980}
    simSpaceSettings {
     dampingAlpha=0
     maxAngularAcceleration=250
     maxAngularVelocity=8
     maxLinearAcceleration=250
     maxLinearVelocity=250
     velocityScaleZ=0.7
     worldAlpha=1
    }
   }
   n7 = RigidBody(ComponentPose=@n6, OverrideWorldGravity=@n5) {
    bOverrideWorldGravity=true
    overridePhysicsAsset=&/F_MED_ShimPilon_Physics_Straps_skydivie
    simulationSpace=BaseBoneSpace
    baseBoneRef {boneName=pelvis}
    overrideWorldGravity {z=-980}
    simSpaceSettings {
     dampingAlpha=0
     maxAngularAcceleration=150
     maxAngularVelocity=8
     maxLinearAcceleration=250
     maxLinearVelocity=250
     velocityScaleZ=0.7
     worldAlpha=1
    }
   }
   n8 = Component To Local(ComponentPose=@n7)
   n9 = Get isfullskydivie() -> (isfullskydivie:bool)
   n10 = "Transform (Modify) Bone - Bone: MSTR_strap_l"(
    ComponentPose=Local To Component(LocalPose=Use cached pose 'Gr/sky_1'())
    Alpha=1
    Translation="0,0,0"
    Rotation="0,-0.07,-10"
    Scale="1,1,1"
   ) {rotationMode=BMM_Additive, rotationSpace=BCS_BoneSpace, boneToModify {boneName=MSTR_strap_l}, rotation {roll=-10, yaw=-0.07}}
   n11 = "Transform (Modify) Bone - Bone: MSTR_strap_r"(ComponentPose=@n10, Alpha=1, Translation="0,0,0", Rotation="-0.006,-0.005,10.000", Scale="1,1,1") {rotationMode=BMM_Additive, rotationSpace=BCS_BoneSpace, boneToModify {boneName=MSTR_strap_r}, rotation=<rot -0.006 -0.0049 10>}
   n12 = Get GravityOverride()
   n13 = RigidBody(ComponentPose=@n11, OverrideWorldGravity=@n12) {
    bOverrideWorldGravity=true
    overridePhysicsAsset=&/F_MED_ShimPilon_Physics_jacet_FE
    simulationSpace=BaseBoneSpace
    baseBoneRef {boneName=pelvis}
    overrideWorldGravity {z=-980}
    simSpaceSettings {dampingAlpha=0, maxAngularAcceleration=180, maxAngularVelocity=3, maxLinearAcceleration=200, maxLinearVelocity=250, worldAlpha=1}
   }
   n14 = RigidBody(ComponentPose=@n13, OverrideWorldGravity=@n12) {
    bOverrideWorldGravity=true
    overridePhysicsAsset=&/F_MED_ShimPilon_Physics_Straps_FE
    simulationSpace=BaseBoneSpace
    baseBoneRef {boneName=pelvis}
    overrideWorldGravity {z=-980}
    simSpaceSettings {dampingAlpha=0, maxAngularAcceleration=150, maxAngularVelocity=8, maxLinearAcceleration=200, maxLinearVelocity=250, worldAlpha=1}
   }
   Output Pose(Result=Blend Poses by bool(
     BlendPose_0=Component To Local(ComponentPose=@n14)
     BlendPose_1=Blend Poses by bool(BlendPose_0=@n8, BlendPose_1=@n4, bActiveValue=@n9)
     bActiveValue=Get bIsInFrontEnd()
    ))
   n15 = Control Rig!(Source=Copy Pose From Mesh() {bUseAttachedParent=true}, Alpha, baseball:bool=Get bIsBaseballSliding(), knee:bool=Get bIsKneeSliding()) -> (Pose) {controlRigAssetReference {blueprintRigClass=&/F_MED_ShimPilon_CtrlRig.F_MED_ShimPilon_CtrlRig_C}}
   Save cached pose 'Gr/sky_1'(Pose=@n15) {cacheName=Gr/sky_1}
  ]}
 EventGraph {nodes [
   Event BlueprintUpdateAnimation()
   > n1 = "Gravity Override State Function (Parameter Asset) (Nativized)"(
    CharacterPartAnimInstance=Self-Reference()
    GameTimeInSeconds=GetGameTimeInSeconds()
    Parameters=&/M_MED_ShimPilon_Parameters
   )
   > Set GravityOverride(GravityOverride=@n1)
   > Set isfullskydivie(isfullskydivie:bool=OR Boolean!(A=Get bIsSkyDiving(), B=Get bIsSkydiveDiveMode(), C:bool=Get bIsSkydiveFloating(), D:bool=Get bIsParachuteOpened()) -> (ReturnValue)) -> (Output_Get:bool)
   TryGetPawnOwner()
  ]}
]
variables [
 Character Part Anim Instance {type=CustomCharacterPartAnimInstance, default=null}
 GravityOverride {type=Vector}
 isfullskydivie {type=boolean}
]
