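# Control Rig for the F_MED_ShimPilon face accessory that procedurally adjusts hood bones to follow head and shoulder poses.
# Weighted parent constraints blend hood_l/r between head, spine_05 and clavicles.
# Head and clavicle angles relative to spine_05 are remapped into corrective offset weights, including a three-way product for combined poses.
# RBAN_OFF selects between two similar chains; the true branch is retuned and adds an opposite-direction head-rotation corrective.
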ControlRigBlueprint /BRCosmetics/Characters/Player/Female/Medium/Bodies/F_MED_ShimPilon/Meshes/Parts/F_MED_ShimPilon_FaceAcc_CtrlRig
graphs [
 Multiply_1_ContainedGraph {nodes [
   n1 = RigVMFunctionEntryNode()
   Multiply = &RigVMFunction_MathDoubleMul(A=@n1.A, B=@n1.B)
   RigVMFunctionReturnNode(Result=&RigVMFunction_MathDoubleMul(B=@n1.C, A=@Multiply.Result).Result)
  ]}
 Multiply_1_ContainedGraph {nodes [
   n1 = RigVMFunctionEntryNode()
   Multiply = &RigVMFunction_MathDoubleMul(A=@n1.A, B=@n1.B)
   RigVMFunctionReturnNode(Result=&RigVMFunction_MathDoubleMul(B=@n1.C, A=@Multiply.Result).Result)
  ]}
 RigVMFunctionLibrary {links [], nodes []}
 RigVMModel {nodes [
   RigUnit_BeginExecution = &RigUnit_BeginExecution()
   RigVMFunction_ControlFlowBranch = &RigVMFunction_ControlFlowBranch(ExecuteContext=@RigUnit_BeginExecution.ExecutePin, Condition=RigVMVariableNode() {variable=RBAN_OFF}.Value)
   &RigUnit_ParentConstraint(
    AdvancedSettings=(InterpolationType=Average,RotationOrderForFilter=XZY)
    Child=(Type=Bone,Name="hood_l")
    Parents=((Item=(Type=Bone,Name="head"),Weight=0.5),(Item=(Type=Bone,Name="spine_05"),Weight=0.2),(Item=(Type=Bone,Name="clavicle_l"),Weight=0.3))
    Weight=1
    ExecutePin=@RigVMFunction_ControlFlowBranch.False
   )
   > &RigUnit_ParentConstraint(
    AdvancedSettings=(InterpolationType=Average,RotationOrderForFilter=XZY)
    Child=(Type=Bone,Name="hood_r")
    Parents=((Item=(Type=Bone,Name="head"),Weight=0.5),(Item=(Type=Bone,Name="spine_05"),Weight=0.2),(Item=(Type=Bone,Name="clavicle_r"),Weight=0.3))
    Weight=1
   )
   n1 = &RigVMFunction_MathDoubleRemap(
    SourceMaximum=60
    SourceMinimum=90
    TargetMaximum=1
    TargetMinimum=0
    Value=&RigVMFunction_MathQuaternionToEuler(
     RotationOrder=XYZ
     Value=&RigUnit_GetRelativeTransformForItem(Child=(Type=Bone,Name="clavicle_l"), Parent=(Type=Bone,Name="spine_05"), bChildInitial=False, bParentInitial=False).RelativeTransform.Rotation
    ).Result.Y
   )
   > &RigUnit_ModifyTransforms(
    ItemToModify=((Item=(Type=Bone,Name="hood_l"),Transform=(Rotation=(X=0,Y=0,Z=0,W=1),Translation=(X=-2.3842,Y=0,Z=-4.051),Scale3D=(X=1,Y=1,Z=1))))
    WeightMinimum=0
    Weight=@n1.Result
   )
   n2 = &RigVMFunction_MathDoubleRemap(
    SourceMaximum=60
    SourceMinimum=90
    TargetMaximum=1
    TargetMinimum=0
    Value=&RigVMFunction_MathQuaternionToEuler(
     RotationOrder=XYZ
     Value=&RigUnit_GetRelativeTransformForItem(Child=(Type=Bone,Name="clavicle_r"), Parent=(Type=Bone,Name="spine_05"), bChildInitial=False, bParentInitial=False).RelativeTransform.Rotation
    ).Result.Y
   )
   > &RigUnit_ModifyTransforms(
    ItemToModify=((Item=(Type=Bone,Name="hood_r"),Transform=(Rotation=(X=0,Y=-0.2164,Z=0,W=0.9763),Translation=(X=-0.2361,Y=0,Z=2),Scale3D=(X=1,Y=1,Z=1))))
    WeightMinimum=0
    Weight=@n2.Result
   )
   n3 = &RigVMFunction_MathQuaternionToEuler(
    RotationOrder=ZYX
    Value=&RigUnit_GetRelativeTransformForItem(Child=(Type=Bone,Name="head"), Parent=(Type=Bone,Name="spine_05"), bChildInitial=False, bParentInitial=False).RelativeTransform.Rotation
   )
   n4 = &RigVMFunction_MathDoubleRemap(SourceMaximum=-50, SourceMinimum=0, TargetMaximum=1, TargetMinimum=0, Value=@n3.Result.Z)
   > &RigUnit_ModifyTransforms(
    ItemToModify="((Item=(Type=Bone,Name=\"hood_m\"),Transform=(Rotation=(X=0,Y=0,Z=0,W=-1),Translation=(X=8.3542,Y=-8,Z=-1),Scale3D=(X=1,Y=1,Z=1))),(Item=(Type=Bone,Name=\"hood_l\"),Transform=(Rotation=(X=0,Y=0,Z=0,W=-1),Translation=(X=-2.7,Y=-2.8,Z=-0.334),Scale3D=(X=1,Y=1,Z= ...(truncated)"
    WeightMinimum=0
    Weight=@n4.Result
   )
   > &RigUnit_ModifyTransforms(
    ItemToModify="((Item=(Type=Bone,Name=\"hood_l\"),Transform=(Rotation=(X=0.471,Y=-0.1822,Z=-0.2055,W=-0.8383),Translation=(X=4.4393,Y=0.868,Z=5.2131),Scale3D=(X=1,Y=1,Z=1))),(Item=(Type=Bone,Name=\"hood_r\"),Transform=(Rotation=(X=0,Y=0,Z=0,W=1),Translation=(X=0,Y=-5.9553,Z= ...(truncated)"
    WeightMinimum=0
    Weight=&RigVMFunction_MathDoubleRemap(SourceMaximum=-65, SourceMinimum=-35, TargetMaximum=1, TargetMinimum=0, Value=@n3.Result.X).Result
   )
   > &RigUnit_ModifyTransforms(
    ItemToModify="((Item=(Type=Bone,Name=\"hood_r\"),Transform=(Rotation=(X=0,Y=0,Z=0,W=-1),Translation=(X=-0.1456,Y=0,Z=-4.8449),Scale3D=(X=1,Y=1,Z=1))),(Item=(Type=Bone,Name=\"hood_l\"),Transform=(Rotation=(X=0,Y=0,Z=-0.1736,W=-0.9848),Translation=(X=-1.6218,Y=0.1619,Z=-3.676 ...(truncated)"
    WeightMinimum=0
    Weight=&RigVMFunction_MathDoubleRemap(SourceMaximum=70, SourceMinimum=40, TargetMaximum=1, TargetMinimum=0, Value=@n3.Result.X).Result
   )
   Multiply_1 = RigVMAggregateNode(C=@n1.Result, A=@n4.Result, B=@n2.Result)
   > &RigUnit_ModifyTransforms(
    ItemToModify="((Item=(Type=Bone,Name=\"hood_l\"),Transform=(Rotation=(X=0,Y=0,Z=0,W=1),Translation=(X=0,Y=0,Z=6),Scale3D=(X=1,Y=1,Z=1))),(Item=(Type=Bone,Name=\"hood_r\"),Transform=(Rotation=(X=0,Y=0,Z=0,W=1),Translation=(X=0,Y=0,Z=6),Scale3D=(X=1,Y=1,Z=1))),(Item=(Type=Bon ...(truncated)"
    WeightMinimum=0
    Weight=@Multiply_1.Result
   )
   &RigUnit_ParentConstraint(
    AdvancedSettings=(InterpolationType=Average,RotationOrderForFilter=XZY)
    Child=(Type=Bone,Name="hood_l")
    Parents=((Item=(Type=Bone,Name="head"),Weight=0.5),(Item=(Type=Bone,Name="spine_05"),Weight=0.2),(Item=(Type=Bone,Name="clavicle_l"),Weight=0.3))
    Weight=1
    ExecutePin=@RigVMFunction_ControlFlowBranch.True
   )
   > &RigUnit_ParentConstraint(
    AdvancedSettings=(InterpolationType=Average,RotationOrderForFilter=XZY)
    Child=(Type=Bone,Name="hood_r")
    Parents=((Item=(Type=Bone,Name="head"),Weight=0.5),(Item=(Type=Bone,Name="spine_05"),Weight=0.2),(Item=(Type=Bone,Name="clavicle_r"),Weight=0.3))
    Weight=1
   )
   RigVMFunction_MathQuaternionToEuler = &RigVMFunction_MathQuaternionToEuler(
    RotationOrder=XYZ
    Value=&RigUnit_GetRelativeTransformForItem(Child=(Type=Bone,Name="clavicle_l"), Parent=(Type=Bone,Name="spine_05"), bChildInitial=False, bParentInitial=False).RelativeTransform.Rotation
   )
   Remap = &RigVMFunction_MathDoubleRemap(SourceMaximum=60, SourceMinimum=90, TargetMaximum=1, TargetMinimum=0, Value=@RigVMFunction_MathQuaternionToEuler.Result.Y)
   > &RigUnit_ModifyTransforms(
    ItemToModify=((Item=(Type=Bone,Name="hood_l"),Transform=(Rotation=(X=0,Y=0,Z=0,W=1),Translation=(X=-2.3842,Y=0,Z=4),Scale3D=(X=1,Y=1,Z=1))))
    WeightMinimum=0
    Weight=@Remap.Result
   )
   n5 = &RigVMFunction_MathDoubleRemap(
    SourceMaximum=60
    SourceMinimum=90
    TargetMaximum=1
    TargetMinimum=0
    Value=&RigVMFunction_MathQuaternionToEuler(
     RotationOrder=XYZ
     Value=&RigUnit_GetRelativeTransformForItem(Child=(Type=Bone,Name="clavicle_r"), Parent=(Type=Bone,Name="spine_05"), bChildInitial=False, bParentInitial=False).RelativeTransform.Rotation
    ).Result.Y
   )
   > &RigUnit_ModifyTransforms(
    ItemToModify=((Item=(Type=Bone,Name="hood_r"),Transform=(Rotation=(X=0,Y=-0.2164,Z=0,W=0.9763),Translation=(X=-0.2361,Y=0,Z=2),Scale3D=(X=1,Y=1,Z=1))))
    WeightMinimum=0
    Weight=@n5.Result
   )
   n6 = &RigVMFunction_MathQuaternionToEuler(
    RotationOrder=ZYX
    Value=&RigUnit_GetRelativeTransformForItem(Child=(Type=Bone,Name="head"), Parent=(Type=Bone,Name="spine_05"), bChildInitial=False, bParentInitial=False).RelativeTransform.Rotation
   )
   n7 = &RigVMFunction_MathDoubleRemap(SourceMaximum=-50, SourceMinimum=10, TargetMaximum=1, TargetMinimum=0, Value=@n6.Result.Z)
   > &RigUnit_ModifyTransforms(
    ItemToModify="((Item=(Type=Bone,Name=\"hood_m\"),Transform=(Rotation=(X=0,Y=0,Z=0,W=-1),Translation=(X=12,Y=-8,Z=-1),Scale3D=(X=1,Y=1,Z=1))),(Item=(Type=Bone,Name=\"hood_l\"),Transform=(Rotation=(X=0,Y=0,Z=0,W=-1),Translation=(X=-2.7,Y=-2.8,Z=-0.334),Scale3D=(X=1,Y=1,Z=1))) ...(truncated)"
    WeightMinimum=0
    Weight=@n7.Result
   )
   > &RigUnit_ModifyTransforms(
    ItemToModify="((Item=(Type=Bone,Name=\"hood_l\"),Transform=(Rotation=(X=0.471,Y=-0.1822,Z=-0.2055,W=-0.8383),Translation=(X=4.4393,Y=0.868,Z=5.2131),Scale3D=(X=1,Y=1,Z=1))),(Item=(Type=Bone,Name=\"hood_r\"),Transform=(Rotation=(X=0,Y=0,Z=0,W=1),Translation=(X=0,Y=-5.9553,Z= ...(truncated)"
    WeightMinimum=0
    Weight=&RigVMFunction_MathDoubleRemap(SourceMaximum=-65, SourceMinimum=-18, TargetMaximum=1, TargetMinimum=0, Value=@n6.Result.X).Result
   )
   > &RigUnit_ModifyTransforms(
    ItemToModify="((Item=(Type=Bone,Name=\"hood_r\"),Transform=(Rotation=(X=0,Y=0,Z=0,W=-1),Translation=(X=-0.1456,Y=0,Z=-4.8449),Scale3D=(X=1,Y=1,Z=1))),(Item=(Type=Bone,Name=\"hood_l\"),Transform=(Rotation=(X=0,Y=0,Z=-0.1736,W=-0.9848),Translation=(X=-1.6218,Y=0.1619,Z=-3.676 ...(truncated)"
    WeightMinimum=0
    Weight=&RigVMFunction_MathDoubleRemap(SourceMaximum=70, SourceMinimum=40, TargetMaximum=1, TargetMinimum=0, Value=@n6.Result.X).Result
   )
   Multiply = RigVMAggregateNode(A=@n7.Result, B=@n5.Result, C=@Remap.Result)
   > &RigUnit_ModifyTransforms(
    ItemToModify="((Item=(Type=Bone,Name=\"hood_l\"),Transform=(Rotation=(X=0,Y=0,Z=0,W=1),Translation=(X=0,Y=0,Z=1.8784),Scale3D=(X=1,Y=1,Z=1))),(Item=(Type=Bone,Name=\"hood_r\"),Transform=(Rotation=(X=0,Y=0,Z=0,W=1),Translation=(X=0,Y=0,Z=5),Scale3D=(X=1,Y=1,Z=1))),(Item=(Typ ...(truncated)"
    WeightMinimum=0
    Weight=@Multiply.Result
   )
   n8 = &RigUnit_GetRelativeTransformForItem(Child=(Type=Bone,Name="head"), Parent=(Type=Bone,Name="spine_05"), bChildInitial=False, bParentInitial=False)
   > &RigUnit_ModifyTransforms(
    ItemToModify="((Item=(Type=Bone,Name=\"hood_m\"),Transform=(Rotation=(X=0,Y=0,Z=0,W=-1),Translation=(X=8.3542,Y=-8,Z=-1),Scale3D=(X=1,Y=1,Z=1))),(Item=(Type=Bone,Name=\"hood_l\"),Transform=(Rotation=(X=0,Y=0,Z=0,W=-1),Translation=(X=1.5548,Y=-0.8368,Z=0.4217),Scale3D=(X=1,Y ...(truncated)"
    WeightMinimum=0
    Weight=&RigVMFunction_MathDoubleRemap(
     SourceMaximum=25
     SourceMinimum=0
     TargetMaximum=1
     TargetMinimum=0
     Value=&RigVMFunction_MathQuaternionToEuler(RotationOrder=ZYX, Value=@n8.RelativeTransform.Rotation).Result.Z
    ).Result
   )
  ]}
]
hierarchy [
 root {loc=<0 0 0>, rot=<0 0 0 1>}
 pelvis {parent=root, loc=<0 -0.5301 98.0571>, rot=<0.0127 0.707 -0.0127 -0.707>}
 spine_01 {parent=pelvis, loc=<6.1208 -0.9328 0>, rot=<0 0 0.034 -0.9994>}
 spine_02 {parent=spine_01, loc=<6.5997 -0.2095 0>, rot=<0 0 0.0238 -0.9997>}
 spine_03 {parent=spine_02, loc=<6.4914 -0.1237 0>, rot=<0 0 0.0149 -0.9999>}
 spine_04 {parent=spine_03, loc=<6.9598 0.0794 0>, rot=<0 0 -0.0132 -0.9999>}
 spine_05 {parent=spine_04, loc=<6.9987 0 0>, rot=<0 0 -0.0015 -1>}
 neck_01 {parent=spine_05, loc=<11.3138 0.1529 0>, rot=<0 0 -0.2003 -0.9797>}
 neck_02 {parent=neck_01, loc=<4.5627 0 0>, rot=<0 0 0.0504 -0.9987>}
 head {parent=neck_02, loc=<5.3914 0 0>, rot=<0 0 0.115 -0.9934>}
 MSTR_haIr {parent=head, loc=<12.5203 11.0019 -0.7474>, rot=<0.3576 -0.6397 -0.5966 -0.327>}
 hair_01 {parent=MSTR_haIr, loc=<0 0 0>, rot=<0 0 0 1>}
 hair_02 {parent=hair_01, loc=<6.7122 0 0>, rot=<0.0356 -0.2237 -0.257 -0.9395>}
 hair_03 {parent=hair_02, loc=<7.6254 0 0>, rot=<0.0578 -0.0028 -0.2559 -0.965>}
 hair_04 {parent=hair_03, loc=<6.0708 0 0>, rot=<-0.0248 -0.0587 0.021 -0.9977>}
 hood_r {parent=head, loc=<-5.7053 1.6402 11.581>, rot=<0.0041 -1 0 0>}
 hood_l {parent=head, loc=<-5.705 1.6402 -11.581>, rot=<0.0041 -1 0 0>}
 hood_m {parent=head, loc=<-3.7343 -8.1497 0>, rot=<0.6457 -0.7636 0 0>}
 hood2_l {parent=head, loc=<4.5361 3.1549 -9.5>, rot=<0.0029 -0.7084 -0.0029 -0.7058>}
 hood2_r {parent=head, loc=<4.462 3.1542 9.5>, rot=<0.7051 0.0064 0.709 -0.0064>}
 MSTR_earring {parent=head, loc=<0.3329 0.763 6.1953>, rot=<0.0041 -1 0 0>}
 earring {parent=MSTR_earring, loc=<0 0 0>, rot=<0 0 0 1>}
 MSTR_haIr2 {parent=head, loc=<12.5532 7.0197 -0.4508>, rot=<0.3576 -0.6397 -0.5966 -0.327>}
 hair2_01 {parent=MSTR_haIr2, loc=<0 0 0>, rot=<0 0 0 1>}
 hair2_02 {parent=hair2_01, loc=<6.7122 0 0>, rot=<0.0121 -0.1715 -0.3238 -0.9304>}
 hair2_03 {parent=hair2_02, loc=<7.6254 0 0>, rot=<0.0578 -0.0028 -0.2559 -0.965>}
 hair2_04 {parent=hair2_03, loc=<6.0708 0 0>, rot=<-0.0248 -0.0587 0.021 -0.9977>}
 hair2_05 {parent=hair2_01, loc=<8.5159 -1.1365 2.8778>, rot=<0.0716 -0.234 0.0591 -0.9678>}
 clavicle_l {parent=spine_05, loc=<5.483 -0.1053 -3.2946>, rot=<0.0465 -0.7289 -0.009 -0.6829>}
 clavicle_r {parent=spine_05, loc=<5.483 -0.1053 3.2946>, rot=<0.7289 0.0465 -0.6829 0.009>}
]
variables [RBAN_OFF {type=bool, default=False}]
