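# Skydive physics asset for the face accessory part of the female medium ShimPilon body, covering two hair bone chains.
# Each chain links four capsule bodies to a small root body on the head, with tight swing limits and near-zero twist between segments.
# Capsules on the head, neck, spine and left clavicle form the surrounding body, joined by free angular constraints.
# The second chain's root also rotates freely on the head, unlike the first.
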PhysicsAsset /BRCosmetics/Characters/Player/Female/Medium/Bodies/F_MED_ShimPilon/Meshes/Parts/F_MED_ShimPilon_FaceAcc_Physics_skydive
bodies {
 MSTR_haIr [{shapeType=Capsule, length=1, radius=1, shapeName=MSTR_haIr_capsule}]
 MSTR_haIr2 [{shapeType=Capsule, length=0.5, radius=0.5, shapeName=MSTR_haIr2_capsule}]
 clavicle_l [
  {shapeType=Capsule, length=8.994, radius=6.687, center=<7.4616 -2.8293 2.7872>, rotation=<rot 89.1655 -121.1788 -112.5631>}
  {shapeType=Capsule, length=6.9763, radius=8.4417, shapeName=clavicle_l_capsule, center=<4.0701 -5.1648 8.7826>, rotation=<rot 0 0 -15.4815>}
 ]
 hair2_01 [{shapeType=Capsule, length=3, radius=2.5, shapeName=hair2_01_capsule, center=<3.4435 -3.05 0>, rotation=<rot -90 -0.0002 0>}]
 hair2_02 [{shapeType=Capsule, length=3, radius=2.5, shapeName=hair2_02_capsule, center=<3.4435 -3.05 1.5>, rotation=<rot -90 0 0>}]
 hair2_03 [{shapeType=Capsule, length=2, radius=2, shapeName=hair2_03_capsule, center=<2.4575 -1.99 1.0714>, rotation=<rot -86.2309 -14.9986 -0.0008>}]
 hair2_04 [{shapeType=Capsule, length=2, radius=2, shapeName=hair2_04_capsule, center=<2.5056 -2.3 0>, rotation=<rot -90 0.0003 0>}]
 hair_01 [{shapeType=Capsule, length=3, radius=3.5, center=<2.7604 -2.8777 -0.73>, rotation=<rot -80.0888 179.997 -179.997>}]
 hair_02 [{
   shapeType=Capsule
   length=3
   radius=3.5
   shapeName=hair_02_capsule
   center=<3.4515 -2.9122 -1.59>
   rotation=<rot 79.9998 -179.9988 -179.9996>
  }]
 hair_03 [{shapeType=Capsule, length=1, radius=2.5, shapeName=hair_03_capsule, center=<1.6323 -0.5988 0>, rotation=<rot 85 179.9991 179.9984>}]
 hair_04 [{shapeType=Capsule, length=1, radius=2.5, shapeName=hair_04_capsule, center=<1.5005 0 0>, rotation=<rot -80 0.0019 0.0001>}]
 head [
  {shapeType=Capsule, length=9.996, radius=8.2833, shapeName=head_capsule, center=<0 2.6331 0.1262>, rotation=<rot -2.2148 -80.0048 -93.2487>}
  {shapeType=Capsule, length=15, radius=6.529, shapeName=head_capsule, center=<-1.5718 -3.5038 -8.8879>, rotation=<rot -90 -0.0002 0>}
 ]
 neck_02 [{shapeType=Capsule, length=1, radius=1}]
 spine_01 [{shapeType=Capsule, length=12.4722, radius=12.1039}]
}
constraints [
 {
  bone1Name=neck_02
  bone2Name=spine_01
  swing1LimitDegrees=45
  swing1Motion=Free
  swing2LimitDegrees=45
  swing2Motion=Free
  twistLimitDegrees=45
  twistMotion=Free
 }
 {bone1Name=hair_01, bone2Name=MSTR_haIr, swing1LimitDegrees=8, swing2LimitDegrees=8, twistLimitDegrees=0}
 {bone1Name=MSTR_haIr, bone2Name=head, swing1LimitDegrees=10, swing2LimitDegrees=10, twistLimitDegrees=0}
 {bone1Name=hair_02, bone2Name=hair_01, swing1LimitDegrees=15, swing2LimitDegrees=15, twistLimitDegrees=0}
 {bone1Name=hair_03, bone2Name=hair_02, swing1LimitDegrees=15, swing2LimitDegrees=15, twistLimitDegrees=0}
 {
  bone1Name=head
  bone2Name=neck_02
  swing1LimitDegrees=45
  swing1Motion=Free
  swing2LimitDegrees=45
  swing2Motion=Free
  twistLimitDegrees=45
  twistMotion=Free
 }
 {bone1Name=hair_04, bone2Name=hair_03, swing1LimitDegrees=10, swing2LimitDegrees=10, twistLimitDegrees=0}
 {
  bone1Name=clavicle_l
  bone2Name=spine_01
  swing1LimitDegrees=45
  swing1Motion=Free
  swing2LimitDegrees=45
  swing2Motion=Free
  twistLimitDegrees=45
  twistMotion=Free
 }
 {
  bone1Name=MSTR_haIr2
  bone2Name=head
  swing1LimitDegrees=45
  swing1Motion=Free
  swing2LimitDegrees=45
  swing2Motion=Free
  twistLimitDegrees=45
  twistMotion=Free
 }
 {bone1Name=hair2_01, bone2Name=MSTR_haIr2, swing1LimitDegrees=5, swing2LimitDegrees=10, twistLimitDegrees=0}
 {bone1Name=hair2_02, bone2Name=hair2_01, swing1LimitDegrees=15, swing2LimitDegrees=15, twistLimitDegrees=0}
 {bone1Name=hair2_03, bone2Name=hair2_02, swing1LimitDegrees=15, swing2LimitDegrees=15, twistLimitDegrees=1}
 {bone1Name=hair2_04, bone2Name=hair2_03, swing1LimitDegrees=10, swing2LimitDegrees=10, twistLimitDegrees=1}
]
