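# Post-process-style AnimBP for ShimPilon's female face-accessory part, adding secondary motion to hair, earring and hood.
# Copies the attached parent's pose through a Control Rig, told when rigid bodies are off, then caches it.
# Two branches chain bone offsets, toggleable RigidBody simulation and spring AnimDynamics, differing in physics asset and hair offsets.
# Skydive, dive, floating or parachute states select the skydive branch; computed GravityOverride is unused by the AnimGraph.
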AnimBlueprint /BRCosmetics/Characters/Player/Female/Medium/Bodies/F_MED_ShimPilon/Meshes/Parts/F_MED_ShimPilon_FaceAcc_AnimBP
target_skeleton=&/F_MED_ShimPilon_FaceAcc_Skeleton
graphs [
 AnimGraph {nodes [
   n1 = Use cached pose 'game/sky'()
   n2 = RigidBody!(
    ComponentPose="Transform (Modify) Bone - Bone: MSTR_haIr2"(ComponentPose=Local To Component(LocalPose=@n1), Alpha=1, Translation="0,0,0", Rotation="5,0,0", Scale="1,1,1") {rotationMode=BMM_Additive, rotationSpace=BCS_BoneSpace, boneToModify {boneName=MSTR_haIr2}, rotation {pitch=5}}
    bAlphaBoolEnabled:bool=Get Rigid Body ON() -> (Rigid Body ON:bool)
    OverrideWorldGravity:Vector="0,0,-980"
    ExternalForce
   ) -> (Pose) {
    alphaInputType=Bool
    bOverrideWorldGravity=true
    overridePhysicsAsset=&/F_MED_ShimPilon_FaceAcc_Physics
    simulationSpace=BaseBoneSpace
    baseBoneRef {boneName=spine_05}
    overrideWorldGravity {z=-980}
    simSpaceSettings {dampingAlpha=0, maxAngularAcceleration=100, maxAngularVelocity=4, maxLinearAcceleration=250, maxLinearVelocity=250, worldAlpha=1}
   }
   n3 = "Transform (Modify) Bone - Bone: MSTR_earring"(ComponentPose=@n2, Alpha=1, Translation="0,0,0", Rotation="-0.000,0.000,25.000", Scale="1,1,1") {rotationMode=BMM_Additive, rotationSpace=BCS_BoneSpace, boneToModify {boneName=MSTR_earring}, rotation {roll=25, yaw=0}}
   n4 = Anim Dynamics - Bone: earring(ComponentPose=@n3, Alpha=1, ExternalForce="0,0,0") {
    angularDampingOverride=0.995
    angularSpringConstant=1100
    bAngularSpring=true
    bOverrideAngularDamping=true
    bUseGravityOverride=true
    simulationSpace=RootRelative
    alphaBoolBlend {blendOption=Linear}
    boundBone {boneName=earring}
    componentAppliedLinearAccClamp=<3000 3000 3000>
    componentLinearAccScale=<0.5 0.5 0.5>
    componentLinearVelScale=<0.5 0.5 0.5>
    gravityOverride {z=-980}
    physicsBodyDefinitions [{
      boundBone {boneName=earring}
      boxExtents=<15 15 15>
      constraintSetup {angularLimitsMax=<10 2 50>, angularLimitsMin=<-7 -5 -2>, angularTarget=<0 0 0>}
      localJointOffset=<5 0 0>
     }]
   }
   n5 = Anim Dynamics - Bone: hood_l(ComponentPose=@n4, Alpha=1, ExternalForce="0,0,0") {
    angularDampingOverride=0.999
    angularSpringConstant=200
    bAngularSpring=true
    bGravityOverrideInSimSpace=true
    bLinearSpring=true
    bOverrideAngularDamping=true
    bOverrideLinearDamping=true
    bUseGravityOverride=true
    linearDampingOverride=1
    linearSpringConstant=150
    simulationSpace=BoneRelative
    boundBone {boneName=hood_l}
    componentAppliedLinearAccClamp=<3000 3000 3000>
    componentLinearAccScale=<0.3 0.3 0.5>
    componentLinearVelScale=<0.3 0.3 0.8>
    gravityOverride {x=-980}
    physicsBodyDefinitions [{
      boundBone {boneName=hood_l}
      boxExtents=<10 10 10>
      constraintSetup {angularLimitsMax=<0 10 0>, angularLimitsMin=<0 -10 0>, angularTarget=<1 0 0>, linearAxesMax=<2 0 2>, linearAxesMin=<-1 0 0>}
      localJointOffset=<0 0 -1>
     }]
    relativeSpaceBone {boneName=spine_05}
    simSpaceSettings {maxAngularAcceleration=2000, maxAngularVelocity=2000, simSpaceAngularAlpha=0.1}
   }
   n6 = Anim Dynamics - Bone: hood_r(ComponentPose=@n5, Alpha=1, ExternalForce="0,0,0") {
    angularDampingOverride=0.999
    angularSpringConstant=200
    bAngularSpring=true
    bGravityOverrideInSimSpace=true
    bLinearSpring=true
    bOverrideAngularDamping=true
    bOverrideLinearDamping=true
    bUseGravityOverride=true
    linearDampingOverride=1
    linearSpringConstant=150
    simulationSpace=BoneRelative
    boundBone {boneName=hood_r}
    componentAppliedLinearAccClamp=<3000 3000 3000>
    componentLinearAccScale=<0.3 0.3 0.5>
    componentLinearVelScale=<0.3 0.3 0.8>
    gravityOverride {x=-980}
    physicsBodyDefinitions [{
      boundBone {boneName=hood_r}
      boxExtents=<11 11 11>
      constraintSetup {angularLimitsMax=<0 10 0>, angularLimitsMin=<0 -10 0>, angularTarget=<1 0 0>, linearAxesMax=<4 0 2>, linearAxesMin=<-1 0 -1>}
      localJointOffset=<0 0 -1>
     }]
    relativeSpaceBone {boneName=spine_05}
    simSpaceSettings {maxAngularAcceleration=2000, maxAngularVelocity=2000, simSpaceAngularAlpha=0.1}
   }
   n7 = Local To Component(LocalPose=Use cached pose 'game/sky'())
   n8 = RigidBody(
    ComponentPose="Transform (Modify) Bone - Bone: MSTR_haIr2"(
     ComponentPose="Transform (Modify) Bone - Bone: MSTR_haIr"(ComponentPose=@n7, Alpha=1, Translation="0,0,0", Rotation="5,0,0", Scale="1,1,1") {rotationMode=BMM_Additive, rotationSpace=BCS_BoneSpace, boneToModify {boneName=MSTR_haIr}, rotation {pitch=5}}
     Alpha=1
     Translation="0,0,0"
     Rotation="5,0,0"
     Scale="1,1,1"
    ) {rotationMode=BMM_Additive, rotationSpace=BCS_BoneSpace, boneToModify {boneName=MSTR_haIr2}, rotation {pitch=5}}
    bAlphaBoolEnabled=Get Rigid Body ON()
    OverrideWorldGravity="0,0,-980"
   ) {
    alphaInputType=Bool
    bOverrideWorldGravity=true
    overridePhysicsAsset=&/F_MED_ShimPilon_FaceAcc_Physics_skydive
    simulationSpace=BaseBoneSpace
    baseBoneRef {boneName=spine_05}
    overrideWorldGravity {z=-980}
    simSpaceSettings {dampingAlpha=0, maxAngularAcceleration=100, maxAngularVelocity=4, maxLinearAcceleration=250, maxLinearVelocity=250, worldAlpha=1}
   }
   n9 = "Transform (Modify) Bone - Bone: MSTR_earring"(ComponentPose=@n8, Alpha=1, Translation="0,0,0", Rotation="-0.000,0.000,25.000", Scale="1,1,1") {rotationMode=BMM_Additive, rotationSpace=BCS_BoneSpace, boneToModify {boneName=MSTR_earring}, rotation=<rot 0 0 25>}
   n10 = Anim Dynamics - Bone: earring(ComponentPose=@n9, Alpha=1, ExternalForce="0,0,0") {
    angularDampingOverride=0.995
    angularSpringConstant=1100
    bAngularSpring=true
    bOverrideAngularDamping=true
    bUseGravityOverride=true
    simulationSpace=RootRelative
    alphaBoolBlend {blendOption=Linear}
    boundBone {boneName=earring}
    componentAppliedLinearAccClamp=<3000 3000 3000>
    componentLinearAccScale=<0.5 0.5 0.5>
    componentLinearVelScale=<0.5 0.5 0.5>
    gravityOverride {z=-980}
    physicsBodyDefinitions [{
      boundBone {boneName=earring}
      boxExtents=<15 15 15>
      constraintSetup {angularLimitsMax=<10 2 50>, angularLimitsMin=<-7 -5 -2>, angularTarget=<0 0 0>}
      localJointOffset=<5 0 0>
     }]
   }
   n11 = Anim Dynamics - Bone: hood_l(ComponentPose=@n10, Alpha=1, ExternalForce="0,0,0") {
    angularDampingOverride=0.999
    angularSpringConstant=200
    bAngularSpring=true
    bGravityOverrideInSimSpace=true
    bLinearSpring=true
    bOverrideAngularDamping=true
    bOverrideLinearDamping=true
    bUseGravityOverride=true
    linearDampingOverride=1
    linearSpringConstant=150
    simulationSpace=BoneRelative
    boundBone {boneName=hood_l}
    componentAppliedLinearAccClamp=<3000 3000 3000>
    componentLinearAccScale=<0.3 0.3 0.5>
    componentLinearVelScale=<0.3 0.3 0.8>
    gravityOverride {x=-980}
    physicsBodyDefinitions [{
      boundBone {boneName=hood_l}
      boxExtents=<10 10 10>
      constraintSetup {angularLimitsMax=<0 10 0>, angularLimitsMin=<0 -10 0>, angularTarget=<1 0 0>, linearAxesMax=<2 0 2>, linearAxesMin=<-1 0 0>}
      localJointOffset=<0 0 -1>
     }]
    relativeSpaceBone {boneName=spine_05}
    simSpaceSettings {maxAngularAcceleration=2000, maxAngularVelocity=2000, simSpaceAngularAlpha=0.1}
   }
   n12 = Anim Dynamics - Bone: hood_r(ComponentPose=@n11, Alpha=1, ExternalForce="0,0,0") {
    angularDampingOverride=0.999
    angularSpringConstant=200
    bAngularSpring=true
    bGravityOverrideInSimSpace=true
    bLinearSpring=true
    bOverrideAngularDamping=true
    bOverrideLinearDamping=true
    bUseGravityOverride=true
    linearDampingOverride=1
    linearSpringConstant=150
    simulationSpace=BoneRelative
    boundBone {boneName=hood_r}
    componentAppliedLinearAccClamp=<3000 3000 3000>
    componentLinearAccScale=<0.3 0.3 0.5>
    componentLinearVelScale=<0.3 0.3 0.8>
    gravityOverride {x=-980}
    physicsBodyDefinitions [{
      boundBone {boneName=hood_r}
      boxExtents=<11 11 11>
      constraintSetup {angularLimitsMax=<0 10 0>, angularLimitsMin=<0 -10 0>, angularTarget=<1 0 0>, linearAxesMax=<4 0 2>, linearAxesMin=<-1 0 -1>}
      localJointOffset=<0 0 -1>
     }]
    relativeSpaceBone {boneName=spine_05}
    simSpaceSettings {maxAngularAcceleration=2000, maxAngularVelocity=2000, simSpaceAngularAlpha=0.1}
   }
   Output Pose(Result=Blend Poses by bool(
     BlendPose_0=Component To Local(ComponentPose=@n12)
     BlendPose_1=Component To Local(ComponentPose=@n6)
     bActiveValue=Get isfullskydivie() -> (isfullskydivie:bool)
    )) {layerGroup=None}
   n13 = Control Rig!(Source=Copy Pose From Mesh() {bUseAttachedParent=true}, Alpha, RBAN_OFF:bool=NOT Boolean(A=Get Rigid Body ON())) -> (Pose) {controlRigAssetReference {blueprintRigClass=&/F_MED_ShimPilon_FaceAcc_CtrlRig.F_MED_ShimPilon_FaceAcc_CtrlRig_C}}
   Save cached pose 'game/sky'(Pose=@n13) {cacheName=game/sky}
  ]}
 EventGraph {nodes [
   Event BlueprintUpdateAnimation()
   > n1 = "Gravity Override State Function (Parameter Asset) (Nativized)"(
    CharacterPartAnimInstance=Self-Reference()
    GameTimeInSeconds=GetGameTimeInSeconds()
    Parameters=/BRCosmetics/Characters/Player/Female/Medium/Bodies/F_MED_ShimPilon/Meshes/M_MED_ShimPilon_Parameters
   )
   > Set GravityOverride(GravityOverride=@n1)
   > Set Rigid Body ON(Rigid Body ON:bool=Get bRigidBodyEnabled()) -> (Output_Get:bool)
   > Set isfullskydivie(isfullskydivie:bool=OR Boolean!(A=Get bIsSkyDiving(), B=Get bIsSkydiveDiveMode(), C:bool=Get bIsSkydiveFloating(), D:bool=Get bIsParachuteOpened()) -> (ReturnValue)) -> (Output_Get:bool)
   Set lod_123(lod_123:bool="integer >= integer"(A=Get CurrentLOD())) -> (Output_Get:bool)
  ]}
]
variables [
 GravityOverride {type=Vector}
 isfullskydivie {type=boolean}
 lod_123 {type=boolean}
 Rigid Body ON {type=boolean}
]
